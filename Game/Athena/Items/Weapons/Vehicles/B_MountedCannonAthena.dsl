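; Mounted cannon weapon Blueprint that plays firing and reload feedback on the ship cannon vehicle the player occupies.
; Both events call the parent, then find the push cannon's skeletal mesh through the instigator pawn.
; Firing attaches a Niagara muzzle effect, stored as NS_FiringDelay, and a reload-start sound.
; Cooldown completion deactivates P_FiringDelay, a different variable from the one set on fire, then spawns a ready-to-fire emitter and reload-end sound.
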(event EventOnPlayWeaponFireFX
  (params (bPersistentFire :type Boolean) (bSecondaryFire :type Boolean))
  ;[1 1 1] Play cooldown fx when fired
  (|CallParentFunction false false)
  ;[1 1 1] Get player, get vehicle, get bone
  (bind _asfort_player_pawn (Utilities|Casting|CastToFortPlayerPawn (Game|GetInstigator)))
  (bind _asfort_athena_skpush_cannon (Utilities|Casting|CastToFortAthenaSKPushCannon (Vehicle|GetVehicleActor _asfort_player_pawn :!class /Script/FortniteGame.FortPlayerPawn)))
  (bind _returnvalue (Niagara|SpawnSystemAttached "/Game/Athena/DrivableVehicles/Ship_Cannon/FX/NS_Ship_Cannon_Vehicle_Launch_Muzzle.NS_Ship_Cannon_Vehicle_Launch_Muzzle" (Variables|/Script/FortniteGame.FortAthenaSKVehicle|GetSkeletalMesh _asfort_athena_skpush_cannon) "Muzzle" "0, 0, 0" "0, 0, 0" "KeepRelativeOffset" true))
  (Variables|Default|SetNS_FiringDelay _returnvalue)
  (Audio|SpawnSoundAttached "/Game/Athena/Sounds/Vehicles/ShipCannon/Cannon_Reload_Start_Cue.Cannon_Reload_Start_Cue" (Variables|/Script/FortniteGame.FortAthenaSKVehicle|GetSkeletalMesh _asfort_athena_skpush_cannon) "Muzzle"))

(event Overheating|EventOnFinishedCooling
  (bind _p_firingdelay (Variables|Default|GetP_FiringDelay))
  ;[1 1 1] Stop firing delay when weapon has cooled down
  (|CallParentFunction)
  (Utilities|IsValid _p_firingdelay
    (:Is%20Valid
      (Components|Activation|SetActive _p_firingdelay false false))
    (:Is%20Not%20Valid
      _))
  ;[1 1 1] Get player, get vehicle, get bone
  (bind _asfort_player_pawn (Utilities|Casting|CastToFortPlayerPawn (Game|GetInstigator)))
  (bind _asfort_athena_skpush_cannon (Utilities|Casting|CastToFortAthenaSKPushCannon (Vehicle|GetVehicleActor _asfort_player_pawn :!class /Script/FortniteGame.FortPlayerPawn)))
  ;[1 1 1] Play ready to fire fx/audio when the cooldown is done
  (Effects|SpawnEmitterAttached "/Game/Effects/Fort_Effects/Gameplay/Cannon/P_Cannon_Fire_ReadyToFire.P_Cannon_Fire_ReadyToFire" (Variables|/Script/FortniteGame.FortAthenaSKVehicle|GetSkeletalMesh _asfort_athena_skpush_cannon) "cannon_pivot")
  (Audio|SpawnSoundAttached "/Game/Athena/Sounds/Vehicles/ShipCannon/Cannon_Reload_End_Cue.Cannon_Reload_End_Cue" (Variables|/Script/FortniteGame.FortAthenaSKVehicle|GetSkeletalMesh _asfort_athena_skpush_cannon) "Muzzle"))

(fn ConstructionScript
  (|CallParentFunction))
