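; Octopus vehicle weapon Blueprint that, on local unequip, guards against a Snowman exploit (FORT-145087).
; If the pawn has the Snowman cue tag, it rechecks every 0.1s while the current weapon is null or a Ferret weapon.
; Any other weapon except a generic utility item triggers removal of Snowman effects and destruction of attached balloon actors.
; It also limits the affiliations to show for to Hostile.
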(event FortWeapon|EventOnLocalUnEquip
  ;[1 1 1] EVENT -> Unequip: Check for Snowman Exploit
  (|CallParentFunction)
  (CallFunction|FerretSnowmanCheck))

(event Custom|FerretSnowmanCheck
  (bind _playerpawn (Variables|Default|GetPlayerPawn))
  (bind _currentweapon (Variables|/Script/FortniteGame.FortPawn|GetCurrentWeapon _playerpawn))
  (bind (_asb_ferret_weapon _bsuccess) (Utilities|Casting|CastToB_FerretWeapon _currentweapon))
  ;[1 1 1] Check to see if they are trying to exploit the poor snowman - Fix for FORT-145087 (For some reason this doesn't work if I run the logic in the parent)
  (if (GameplayTags|HasMatchingGameplayTag (Variables|Default|GetPlayerPawn) (Variables|Default|GetSnowmanCue))
    (if (or _bsuccess (Utilities|Equal(Object) _currentweapon 0))
      (Utilities|Time|SetTimerbyEvent (EventDispatchers|CreateEvent :!event Custom|FerretSnowmanCheck) 0.1)
      (else
        (Utilities|Casting|CastToB_UtilityItem_Generic _currentweapon
          (:CastFailed
            (GameplayEffects|RemoveActiveEffectswithGrantedTags (Ability|GetAbilitySystemComponent _playerpawn) "(GameplayTags=((TagName=\"Athena.Item.Snowman\")))")
            (for _array_element (Actor|GetAttachedActors _playerpawn true false)
              (bind _asbuilding_gameplay_actor_balloon (Utilities|Casting|CastToBuildingGameplayActorBalloon _array_element))
              (Utilities|IsValid _asbuilding_gameplay_actor_balloon
                (:Is%20Valid
                  (Actor|DestroyActor _asbuilding_gameplay_actor_balloon))))))))))

(fn GetAffiliationsToShowFor
  (return (Utilities|Array|MakeArray "Hostile")))

(fn ConstructionScript
  (|CallParentFunction))
