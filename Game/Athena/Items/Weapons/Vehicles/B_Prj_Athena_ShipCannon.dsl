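; Projectile for the Athena ship cannon: a cannonball that starts at the push cannon's muzzle, punches through weaker buildings and explodes.
; On the server, hits on new, damageable, low-max-health buildings reduce saved velocity, resuming movement via timer while above minimum speed; otherwise it explodes.
; Entering water also detonates it.
; Explosions launch players server-side, while clients play Niagara effects and sound at a recent hit, water surface or actor location.
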(event OnProjectileStop(ProjectileMovementComponent)
  (params (ImpactResult :type Hit%20Result%20Structure%20%28by%20ref%29))
  ;[1 1 1] Whenever you hit something, if its a wall or player, check the hit limit, and keep moving if you pass
  (bind _output_get (Variables|Default|SetLastHitResult ImpactResult))
  (Variables|Default|SetLastHitServerTime (Fortnite|Sync|Time|GetServerWorldTimeSeconds))
  (Networking|SwitchHasAuthority
    (:Authority
      (bind _asbuilding_actor (Utilities|Casting|CastToBuildingActor (.HitActor (Collision|BreakHitResult _output_get)))
        (:then
          (if (CallFunction|ShouldResimulateAfterBuildingHit _asbuilding_actor)
            (Variables|Default|SetLastHitBuilding _asbuilding_actor)
            (Variables|Default|SetStartingVelocity (CallFunction|CalculateNewVelocity (.TraceStart (Collision|BreakHitResult _output_get)) (.TraceEnd (Collision|BreakHitResult _output_get))))
            (if (.ReturnValue2 (CallFunction|CalculateNewVelocity (.TraceStart (Collision|BreakHitResult _output_get)) (.TraceEnd (Collision|BreakHitResult _output_get))))
              (CallFunction|HandleProjectileStop)
              (else
                _))
            (else
              _)))
        (:CastFailed
          _))
      (Delivery|DoExplosion))))

(event OnEnterWaterEvent(WaterInteractionComponent)
  (params (WaterBody :type Fort%20Water%20Body%20Actor%20Object%20Reference) (WaterInteractionComponent :type Fort%20Water%20Interaction%20Component%20Object%20Reference) (bIsFirstBody :type Boolean))
  ;[0 0.6148 1] EVENT -> On Enter Water
  (if bIsFirstBody
    (bind _output_get (Variables|Default|SetWater%20Body WaterBody))
    (Variables|Default|SetWater%20Interaction%20Component WaterInteractionComponent)
    (Variables|Default|SetWater%20Surface%20Location (WaterBody|GetWaterSurfaceInfoatLocation (Variables|/Script/Water.WaterBody|GetWaterBodyComponent _output_get) (Transformation|GetWorldLocation (Variables|Collision|GetCapsuleComponent)) false))
    (Variables|Default|SetHit%20Water true)
    (Delivery|DoExplosion)
    (Rendering|SetActorHiddenInGame true)))

(event EventOnExploded
  (params (HitActors :type Array%20of%20Actor%20Object%20References) (HitResults :type Array%20of%20Hit%20Result%20Structures))
  (Utilities|FlowControl|Sequence
    (:then_0
      ;[1 1 1] Knockback players that were only hit by the explosion
      (Networking|SwitchHasAuthority
        (:Authority
          (for _array_element HitActors
            (bind _asplayer_pawn_athena (Utilities|Casting|CastToPlayerPawn_Athena _array_element))
            (CallFunction|LaunchPawn _asplayer_pawn_athena)))))
    (:then_1
      ;[0.6417 0 1] VFX & Sound (Client Side Only)
      (Networking|CanExecuteCosmeticEvents
        (:True
          (CallFunction|PlayExplodeVFXAndSound (select (<= (- (Fortnite|Sync|Time|GetServerWorldTimeSeconds) (Variables|Default|GetLastHitServerTime)) (Variables|Default|GetMaxRollBackExplosionVFXTime%20Const)) (Variables|Default|GetLastHitResult) (Collision|MakeHitResult false))))))))

(event EventBeginPlay
  ;[1 1 1] After you have spawned, save off your velocity for later use
  (bind _asfort_player_pawn (Utilities|Casting|CastToFortPlayerPawn (Variables|Actor|GetInstigator)))
  (Collision|IgnoreActorwhenMoving (Variables|Collision|GetCapsuleComponent) (Vehicle|GetVehicleActor _asfort_player_pawn :!class /Script/FortniteGame.FortPlayerPawn) true)
  ;; Set Start Position
  (bind _asfort_athena_skpush_cannon (Utilities|Casting|CastToFortAthenaSKPushCannon (Vehicle|GetVehicleActor _asfort_player_pawn :!class /Script/FortniteGame.FortPlayerPawn))
    (:then
      (Transformation|SetActorLocation (Transformation|GetSocketLocation (Variables|/Script/FortniteGame.FortAthenaVehicle|GetSkeletalMesh _asfort_athena_skpush_cannon) "Muzzle") false true))
    (:CastFailed
      _))
  (Utilities|FlowControl|Delay 0.01)
  (Variables|Default|SetStartingVelocity (Variables|/Script/Engine.MovementComponent|GetVelocity (Variables|Movement|GetProjectileMovementComponent))))

(event EventOnTouched
  (params (OtherActor :type Actor%20Object%20Reference) (OtherComp :type Primitive%20Component%20Object%20Reference) (HitResult :type Hit%20Result%20Structure%20%28by%20ref%29) (bIsOverlap :type Boolean))
  (bind (_bblockinghit _binitialoverlap _time _distance _location _impactpoint _normal _impactnormal _physmat _hitactor) (Collision|BreakHitResult HitResult))
  (bind (_asfort_player_pawn _bsuccess) (Utilities|Casting|CastToFortPlayerPawn _hitactor))
  (if (not bIsOverlap)
    (Utilities|FlowControl|Sequence
      (:then_0
        (Audio|PlaySoundatLocation (select _bsuccess (Variables|Default|GetCannonball_PlayerImpactSound) (Variables|Default|GetCannonball_SurfaceImpactSound)) _location))
      (:then_1
        (Networking|CanExecuteCosmeticEvents
          (:True
            (if (not (or (Fortnite|IsMobilePlatform) _bsuccess))
              (bind _returnvalue (Rendering|Decal|SpawnDecalatLocation "/Game/Weapons/Materials/MD_Decal_BulletHole_Lifetime_Inst.MD_Decal_BulletHole_Lifetime_Inst" "14.000000,22.000000,22.000000" _location (Math|Conversions|RotationFromXVector (* _impactnormal -1.0)) 4.5))
              (Rendering|Components|Decal|SetFadeOut _returnvalue 0.0 4.5 true))))))))

(fn CalculateNewVelocity
  (params (TraceStart :type Vector) (TraceEnd :type Vector))
  (return (* (Math|Vector|Normalize (- TraceEnd TraceStart)) (* (Math|Vector|VectorLength (Variables|Default|GetStartingVelocity)) (ScalableFloat|GetValueatLevel (Variables|Default|GetVelocityLossMultiplier)))) (>= (* (Math|Vector|VectorLength (Variables|Default|GetStartingVelocity)) (ScalableFloat|GetValueatLevel (Variables|Default|GetVelocityLossMultiplier))) (ScalableFloat|GetValueatLevel (Variables|Default|GetMinimumVelocity)))))

(fn HandleProjectileStop
  (Networking|SwitchHasAuthority
    (:Authority
      (Utilities|Time|SetTimerbyEvent (EventDispatchers|CreateEvent :!event Resume_Movement) 1e-05))))

(fn LaunchPawn
  (params (LaunchedPawn :type Fort%20Player%20Pawn%20Athena%20Object%20Reference))
  (Character|LaunchCharacter LaunchedPawn (Math|Vector|MakeVector (.x (* (Math|Vector|Normalize (- (Transformation|GetActorLocation LaunchedPawn) (Transformation|GetActorLocation))) (ScalableFloat|GetValueatLevel (Variables|Default|GetEnemyLaunchVelocity)))) (.y (* (Math|Vector|Normalize (- (Transformation|GetActorLocation LaunchedPawn) (Transformation|GetActorLocation))) (ScalableFloat|GetValueatLevel (Variables|Default|GetEnemyLaunchVelocity)))) (Math|Float|Max(Float) (.z (* (Math|Vector|Normalize (- (Transformation|GetActorLocation LaunchedPawn) (Transformation|GetActorLocation))) (ScalableFloat|GetValueatLevel (Variables|Default|GetEnemyLaunchVelocity)))) 5e+02)) true true))

(fn PlayExplodeVFXAndSound
  (params (HitResult :type Hit%20Result%20Structure))
  (bind (_bblockinghit _binitialoverlap _time _distance _location _impactpoint _normal) (Collision|BreakHitResult HitResult))
  (bind (_returnvalue_location _returnvalue_rotation) (Math|Transform|BreakTransform (Transformation|GetActorTransform)))
  (Networking|CanExecuteCosmeticEvents
    (:True
      (bind _returnvalue (Niagara|SpawnSystematLocation "/Game/Athena/DrivableVehicles/Ship_Cannon/FX/NS_Cannon_Ball_Impact.NS_Cannon_Ball_Impact" (select (Variables|Default|GetHit%20Water) (Variables|Default|GetWater%20Surface%20Location) (select _bblockinghit _location _returnvalue_location)) (select _bblockinghit _returnvalue_rotation (Math|Rotator|MakeRotfromZ _normal))))
      (Niagara|SetNiagaraVariable(Bool) _returnvalue "Hit Water" (Variables|Default|GetHit%20Water))
      (Collision|SetCollisionEnabled (Variables|Collision|GetCapsuleComponent) "NoCollision")
      (Components|Activation|Deactivate (Variables|Default|GetNS_Cannon_Prj_Trail))
      (Development|SetHiddeninGame (Variables|Default|GetSphere) true false)
      (Utilities|FlowControl|DoOnce false
        (:Completed
          (Audio|PlaySoundatLocation (Variables|Default|GetCannonball_ExplosionSound) (Transformation|GetActorLocation)))))))

(fn Resume_Movement
  (Components|Movement|SetUpdatedComponent (Variables|Movement|GetProjectileMovementComponent) (Variables|Collision|GetCapsuleComponent))
  (Variables|/Script/Engine.MovementComponent|SetVelocity (Variables|Default|GetStartingVelocity) (Variables|Movement|GetProjectileMovementComponent)))

(fn ShouldResimulateAfterBuildingHit
  (params (HitBuilding :type Building%20Actor%20Object%20Reference))
  (return (Math|Boolean|ANDBoolean (<= (Health|GetMaxHealth HitBuilding :!class /Script/FortniteGame.BuildingActor) 1e+03) (Variables|/Script/Engine.Actor|GetbCanBeDamaged HitBuilding) (Utilities|NotEqual(Object) HitBuilding (Variables|Default|GetLastHitBuilding)))))

(fn ConstructionScript
  (bind _projectilemovementcomponent (Variables|Movement|GetProjectileMovementComponent))
  ;[1 1 1] Establish table data for max speed and gravity
  (Variables|/Script/Engine.ProjectileMovementComponent|SetInitialSpeed (ScalableFloat|GetValueatLevel (Variables|Default|GetCannonballInitialSpeed)) _projectilemovementcomponent)
  (Variables|/Script/Engine.ProjectileMovementComponent|SetProjectileGravityScale (ScalableFloat|GetValueatLevel (Variables|Default|GetCannonballGravityScale)) _projectilemovementcomponent))
