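; Weapon Blueprint for a generic vehicle-mounted ranged cannon that hides its weapon mesh on BeginPlay and drives firing, reload and targeting presentation.
; The construction script configures muzzle, reload, wind and shell particles, which fire and reload events toggle, with a failsafe timer stopping stuck looping effects.
; Cosmetic alterations async-load replacement effects.
; Scope targeting swaps pawn, mesh and reticle visibility, animates post-process and can mark AI via custom depth.
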(event EventBeginPlay
  (Rendering|SetVisibility (Variables|Visual|GetWeaponMesh) false false) ; Hide the weapon mesh
  ;[0 0 0] Send wind particle reference to the wind manager
  (if (Variables|Default|GetIs%20Wind%20Enabled)
    (Wind|AddWindParticleSystemComponent (Fortnite|GetWindManager) (Variables|Effects|GetMuzzleWindParticle(Empty)))))

(event Custom|ShellsON_(onPump)
  ;[0.3841 0.6674 1] Shells VFX (on Pump)
  (CallFunction|ActivateShellsFX (and (Variables|Effects%28Shells%29|GetUseShellsOnPump?) (> (Rendering|GetRenderingDetailMode) 0))))

(event EventOnPlayWeaponFireFX
  (params (bPersistentFire :type Boolean) (bSecondaryFire :type Boolean))
  ;[0.4433 0 1] Muzzle Flash VFX
  (Utilities|FlowControl|Sequence
    (:then_0
      (CallFunction|MuzzleFlashFX bPersistentFire)
      ;;[1 0.09001 0.04025] Failsafe hack to stop looping FX if we exceed the RPC limit and don't properly remove the gameplay cue
      (if (Effects|ShouldPlayPersistentFireFX)
        (Utilities|Time|SetTimerbyFunctionName 0 "StopWeaponFireFX" (* (* (FortWeapon|GetLocalMagazineAmmoCount) (FortWeapon|GetTimetoNextFire)) 1.1))))
    (:then_1
      (CallFunction|ActivateorDeactivateWindParticle true))
    (:then_2
      (if (and (not (Weapon|IsUsingScopeorFirstPersonCamera)) (> (Rendering|GetRenderingDetailMode) 0))
        (CallFunction|ActivateShellsFX (Variables|Effects%28Shells%29|GetUseShellsOnFire?))))
    (:then_3
      (FortWeapon|PlayForceFeedback false bPersistentFire))))

(event EventOnStopWeaponFireFX
  (bind _returnvalue (Utilities|Time|GetGameTimeinSeconds))
  ;[0.4433 0 1] Muzzle Flash VFX
  (Utilities|Time|ClearTimerbyFunctionName 0 "StopWeaponFireFX")
  ;[1 1 1] Enforce a short minimum time to play looping FX
  (if (not (>= (- _returnvalue (Variables|Effects|GetLastPlayFXTime)) (Variables|Effects|GetMinPlayFXTime)))
    (Utilities|FlowControl|RetriggerableDelay (- (Variables|Effects|GetMinPlayFXTime) (- _returnvalue (Variables|Effects|GetLastPlayFXTime)))))
  (CallFunction|ActivateorDeactivateWindParticle false)
  (CallFunction|DeactivateMuzzleFX)
  (Effects|DeactivateBeamFX)
  (CallFunction|DeactivateShellsFX)
  (FortWeapon|StopForceFeedback))

(event EventOnInitCosmeticAlterations
  (params (CosmeticMod :type Fort%20Cosmetic%20Modification%20Structure))
  (bind (_cosmeticmaterial _ambientparticlesystem _ambientniagarasystem _muzzleparticlesystem _muzzleniagarasystem _reloadparticlesystem _reloadniagarasystem _beamparticlesystem _beamniagarasystem _tracertemplate _bmodifycolor _coloralteration _colorparametername _bmodifydecalcolour _decalcolouralterationstart _decalcolouralterationend _bmodifyshellcolour _shellcolouralteration) (Utilities|Struct|BreakFortCosmeticModification CosmeticMod))
  (bind (_r _g _b) (Utilities|Struct|BreakLinearColor _shellcolouralteration))
  ;[0.06096 0.8755 1] Setup Alteration FX
  ;; Alteration Logic
  (Networking|CanExecuteCosmeticEvents
    (:True
      (Utilities|FlowControl|Sequence
        (:then_0
          (if (not (Asset|IsAssetNull _ambientniagarasystem))
            (bind _object (Utilities|AsyncLoadAsset _ambientniagarasystem)
              (:Completed
                (bind _asniagara_system (Utilities|Casting|CastToNiagaraSystem _object))
                (bind _alteration_ambient_ps (Variables|Effects%28Alterations%29|GetAlteration%20Ambient%20PS)
                  (:then
                    (Components|DestroyComponent _alteration_ambient_ps))
                  (:else
                    _))
                (bind _niagara_component_system_reference (Utilities|ElementalWeaponSetupFX (Variables|Visual|GetWeaponMesh) _asniagara_system)
                  (:Out
                    (Variables|Effects%28Alterations%29|SetAlteration%20Ambient%20PS _niagara_component_system_reference)))))))
        (:then_1
          (if (not (Asset|IsAssetNull _muzzleparticlesystem))
            (bind _object_2 (Utilities|AsyncLoadAsset _muzzleparticlesystem)
              (:Completed
                (bind _asparticle_system (Utilities|Casting|CastToParticleSystem _object_2))
                (Effects|Components|ParticleSystem|SetTemplate (Variables|Effects|GetMuzzle(Empty)) _asparticle_system)))))
        (:then_2
          (if (and (Effects|GetUseBeamParticles) (not (Asset|IsAssetNull _beamparticlesystem)))
            (bind _object_3 (Utilities|AsyncLoadAsset _beamparticlesystem)
              (:Completed
                (bind _asparticle_system_2 (Utilities|Casting|CastToParticleSystem _object_3))
                (bind _returnvalue (Effects|SetBeamParticleSystem _asparticle_system_2))
                (Effects|Components|ParticleSystem|SetTemplate (CallFunction|GetBeamPSC) _returnvalue)))))
        (:then_3
          (if (not (Asset|IsAssetNull _reloadparticlesystem))
            (bind _object_4 (Utilities|AsyncLoadAsset _reloadparticlesystem)
              (:Completed
                (bind _asparticle_system_3 (Utilities|Casting|CastToParticleSystem _object_4))
                (Effects|Components|ParticleSystem|SetTemplate (Variables|Default|GetReload(Empty)) _asparticle_system_3)))))
        (:then_4
          (if _bmodifydecalcolour
            (Decals|SetDecalColorStart _decalcolouralterationstart)
            (Decals|SetDecalColorEnd _decalcolouralterationend)))
        (:then_5
          (if _bmodifyshellcolour
            (Effects|Components|ParticleSystem|SetFloatParameter (Variables|Effects%28Shells%29|GetShells%28empty%29) "Red" _r)
            (Effects|Components|ParticleSystem|SetFloatParameter (Variables|Effects%28Shells%29|GetShells%28empty%29) "Green" _g)
            (Effects|Components|ParticleSystem|SetFloatParameter (Variables|Effects%28Shells%29|GetShells%28empty%29) "Blue" _b)))))))

(event EventOnPlayReloadFX
  (params (ReloadStage :type EFortReloadFXState%20Enum :default "ReloadStart"))
  ;[0.3518 0.0375 0.63] Use Reload Smoke Trails
  (CallFunction|MuzzlePlayReloadFX ReloadStage)
  (FortWeapon|PlayForceFeedback true))

(event EventOnInitAlteration
  (params (NewAlteration :type Fort%20Alteration%20Item%20Definition%20Object%20Reference)))

(event FortWeapon|EventOnEquippedAndAttached
  (bind _scopemesh1p (Variables|Components|GetScopeMesh1P))
  ;[1 1 1] Init Scope Variables
  (bind _asfort_pawn (Utilities|Casting|CastToFortPawn (Game|GetInstigator)))
  (Utilities|IsValid (Pawn|GetLocalViewingPlayerController _asfort_pawn)
    (:Is%20Valid
      (Networking|CanExecuteCosmeticEvents
        (:True
          (if (Scope|UseScopeTargeting)
            (Utilities|IsValid _scopemesh1p
              (:Is%20Valid
                (Rendering|Material|SetVectorParameterValue "/Game/Packages/Fortress_SharedMaterials/GlobalMaterialParameters/FortniteMaterialParameters.FortniteMaterialParameters" "TopLeftScope" (Math|Conversions|ToLinearColor(Vector) (Transformation|GetSocketLocation _scopemesh1p "TopLeftScope")) :!class /Script/Engine.KismetMaterialLibrary)
                (Rendering|Material|SetVectorParameterValue "/Game/Packages/Fortress_SharedMaterials/GlobalMaterialParameters/FortniteMaterialParameters.FortniteMaterialParameters" "BottomRightScope" (Math|Conversions|ToLinearColor(Vector) (Transformation|GetSocketLocation _scopemesh1p "BottomRightScope")) :!class /Script/Engine.KismetMaterialLibrary)))))))
    (:Is%20Not%20Valid
      (CallFunction|HideIronSightsMesh))))

(event Custom|InitializeScopeVariables
  (bind _scopemesh1p (Variables|Components|GetScopeMesh1P))
  ;[1 1 1] Init Scope Variables
  (bind _asfort_pawn (Utilities|Casting|CastToFortPawn (Game|GetInstigator)))
  (Utilities|IsValid (Pawn|GetLocalViewingPlayerController _asfort_pawn)
    (:Is%20Valid
      (Networking|CanExecuteCosmeticEvents
        (:True
          (if (Scope|UseScopeTargeting)
            (Utilities|IsValid _scopemesh1p
              (:Is%20Valid
                (Rendering|Material|SetVectorParameterValue "/Game/Packages/Fortress_SharedMaterials/GlobalMaterialParameters/FortniteMaterialParameters.FortniteMaterialParameters" "TopLeftScope" (Math|Conversions|ToLinearColor(Vector) (Transformation|GetSocketLocation _scopemesh1p "TopLeftScope")) :!class /Script/Engine.KismetMaterialLibrary)
                (Rendering|Material|SetVectorParameterValue "/Game/Packages/Fortress_SharedMaterials/GlobalMaterialParameters/FortniteMaterialParameters.FortniteMaterialParameters" "BottomRightScope" (Math|Conversions|ToLinearColor(Vector) (Transformation|GetSocketLocation _scopemesh1p "BottomRightScope")) :!class /Script/Engine.KismetMaterialLibrary)))))))
    (:Is%20Not%20Valid
      (CallFunction|HideIronSightsMesh))))

(event EventOnSetTargeting
  (params (bNewIsTargeting :type Boolean))
  (bind _weaponmesh (Variables|Visual|GetWeaponMesh))
  (bind _scopemesh1p (Variables|Components|GetScopeMesh1P))
  ;[1 1 1] Scoped weapon targeting downsights
  (Networking|CanExecuteCosmeticEvents
    (:True
      (bind _asfort_pawn (Utilities|Casting|CastToFortPawn (Game|GetInstigator)))
      (Utilities|IsValid (Pawn|GetLocalViewingPlayerController _asfort_pawn)
        (:Is%20Valid
          (if (Scope|UseScopeTargeting)
            (if bNewIsTargeting
              (CallFunction|InitializeScopeVariables)
              (Utilities|FlowControl|Sequence
                (:then_0
                  (Utilities|FlowControl|Delay 0.2)
                  (if (Camera|GetIsTargeting)
                    (FortPawn|SetPawnVisibilityLocalControllerOnly _asfort_pawn true false)
                    (FortWeapon|SetWeaponVisibilityControllerOnly true true)
                    (FortWeapon|SetHiddeninGameControllerOnly _weaponmesh true)
                    (Development|SetHiddeninGame _scopemesh1p false false)
                    (FortWeapon|SetHiddeninGameOtherLocalControllersOnly _scopemesh1p true)
                    (CallFunction|HideReticle)
                    ;;[1 1 1] Custom Depth
                    (if (Variables|Scope|GetScope%20-%20Render%20Enemies%20To%20Custom%20Depth%20Buffer)
                      (CallFunction|UpdateEnemyCustomDepths true 0))))
                (:then_1
                  (Utilities|FlowControl|Delay 0.1)
                  (if (Camera|GetIsTargeting)
                    (Utilities|FlowControl|Sequence
                      (:then_0
                        ;;[1 1 1] Scope Post Process
                        (bind (_direction _downsightpostprocessamount) (Timeline|AnimateScopePostProcess Play)
                          (:Update
                            (Variables|Scope|SetScopePostProcessBlendWeight _downsightpostprocessamount)
                            (Variables|Scope|SetScopePostProcessEnabled true))
                          (:Finished
                            (if (Utilities|Enum|Equal(Enum) _direction "Backward")
                              (Development|SetHiddeninGame (Variables|Components|GetScopeMesh1P) true false)
                              (Variables|Scope|SetScopePostProcessEnabled false)
                              (Variables|Scope|SetScopePostProcessBlendWeight 0.0)
                              ;;;[1 1 1] Custom Depth
                              (if (Variables|Scope|GetScope%20-%20Render%20Enemies%20To%20Custom%20Depth%20Buffer)
                                (CallFunction|UpdateEnemyCustomDepths false 0))))))
                      (:then_1
                        _)))))
              (else
                (Utilities|FlowControl|Sequence
                  (:then_0
                    (if (not (Weapon|IsUsingFirstPersonCamera))
                      (bind _asfort_pawn_2 (Utilities|Casting|CastToFortPawn (Game|GetInstigator)))
                      (FortPawn|SetPawnVisibilityLocalControllerOnly _asfort_pawn_2 false false)
                      (FortWeapon|SetHiddeninGameControllerOnly _weaponmesh))
                    (CallFunction|ShowReticle)
                    (bind _asfort_player_controller_zone (Utilities|Casting|CastToFortPlayerControllerZone (Game|GetInstigatorController)))
                    (Class|FortPlayerController|FixupInteractionWidgetsonUnzoom _asfort_player_controller_zone))
                  (:then_1
                    (Audio|SpawnSoundAttached (Variables|Sounds|GetSound_ScopeZoomOut) _weaponmesh)
                    (Timeline|AnimateScopePostProcess Reverse))
                  (:then_2
                    _))))
            (Utilities|FlowControl|DoOnce false
              (:Completed
                (Audio|SpawnSoundAttached (Variables|Sounds|GetSound_ScopeZoomIn) _weaponmesh)))
            (else
              (Utilities|HandleFirstPersonTargeting bNewIsTargeting))))
        (:Is%20Not%20Valid
          (CallFunction|HideIronSightsMesh))))))

(event FortWeapon|EventOnLocalUnEquip
  (bind _weaponmesh (Variables|Visual|GetWeaponMesh))
  ;[1 1 1] Scoped weapon targeting downsights
  (bind _asfort_pawn (Utilities|Casting|CastToFortPawn (Game|GetInstigator)))
  (Utilities|IsValid (Pawn|GetLocalViewingPlayerController _asfort_pawn)
    (:Is%20Valid
      (if (Scope|UseScopeTargeting)
        (Utilities|FlowControl|Sequence
          (:then_0
            (if (not (Weapon|IsUsingFirstPersonCamera))
              (bind _asfort_pawn_2 (Utilities|Casting|CastToFortPawn (Game|GetInstigator)))
              (FortPawn|SetPawnVisibilityLocalControllerOnly _asfort_pawn_2 false false)
              (FortWeapon|SetHiddeninGameControllerOnly _weaponmesh))
            (CallFunction|ShowReticle)
            (bind _asfort_player_controller_zone (Utilities|Casting|CastToFortPlayerControllerZone (Game|GetInstigatorController)))
            (Class|FortPlayerController|FixupInteractionWidgetsonUnzoom _asfort_player_controller_zone))
          (:then_1
            (Audio|SpawnSoundAttached (Variables|Sounds|GetSound_ScopeZoomOut) _weaponmesh)
            ;;[1 1 1] Scope Post Process
            (Timeline|AnimateScopePostProcess Reverse))
          (:then_2
            (Utilities|FlowControl|DoOnce false
              (:Completed
                (Audio|SpawnSoundAttached (Variables|Sounds|GetSound_ScopeZoomIn) _weaponmesh)))))
        (else
          (CallFunction|ShowReticle))))
    (:Is%20Not%20Valid
      (CallFunction|HideIronSightsMesh))))

(event EventOnWeaponVisibilityChanged
  (params (bVisible :type Boolean) (bSetForLocalControllerOnly :type Boolean))
  ;[1 1 1] EVENT -> OnWeaponVisibilityChanged
  (Development|PrintString (Utilities|String|BuildString(Boolean) (Utilities|GetDisplayName (Variables|Visual|GetWeaponMesh)) "
Weapon Visibility Changed: " bVisible) false false))

(event Custom|Update%20Enemy%20Custom%20Depths
  (params (Enable%20Or%20Disable :type Boolean :default true) (StencilBufferValue :type Integer :default 3))
  ;[1 1 1] Custom Depth
  (bind _outactors (Actor|GetAllActorsOfClass "/Script/FortniteGame.FortAIPawn"))
  (bind _output_get (Variables|Scope|SetArray%20Of%20Active%20Enemy%20AI _outactors))
  (for _array_element _output_get
    (Utilities|IsValid (Variables|/Script/Engine.Character|GetMesh _array_element)
      (:Is%20Valid
        (Rendering|SetRenderCustomDepth (Variables|/Script/Engine.Character|GetMesh _array_element) Enable%20Or%20Disable)
        (Rendering|SetCustomDepthStencilValue (Variables|/Script/Engine.Character|GetMesh _array_element) StencilBufferValue)))))

(event EventOnEquippedWeaponDestory
  ;[0.725 0.6854 0.4277] Show fx for weapon destruction on depleted durability
  (if (Variables|Effects|GetUseDestroyEffect)
    (Effects|SpawnEmitteratLocation (Variables|Effects|GetWeaponDurabilityDestroyEffect) (Transformation|GetSocketLocation (Variables|Visual|GetWeaponMesh) "Handle"))
    (bind _asfort_pawn (Utilities|Casting|CastToFortPawn (Game|GetInstigator)))
    (bind _returnvalue (Effects|SpawnEmitterAttached :EmitterTemplate (Variables|Effects|GetWeaponDurabilityDestroyEffectIcon) :AttachToComponent (Variables|/Script/Engine.Character|GetMesh _asfort_pawn) :AttachPointName "Root_Effect" :Location "0, 0,130.000000" :Scale "1.000000,1.000000,1.000000" :LocationType "SnapToTarget" :bAutoDestroy true :PoolingMethod "None" :bAutoActivate true))
    (bind _returnvalue_2 (Rendering|Material|CreateDynamicMaterialInstance _returnvalue 0 "/Game/Effects/WeaponDestroy/M_WeaponDestroyExclamation.M_WeaponDestroyExclamation" "None" :!class /Script/Engine.PrimitiveComponent))
    (Rendering|Material|SetMaterial _returnvalue 0 _returnvalue_2)
    (Rendering|Material|SetTextureParameterValue _returnvalue_2 "IconTex" (ItemPreviewImage|GetorLoadSynchronously (Item|GetWeaponDataSmallPreviewImage)))))

(event Custom|SetWeaponPierceThrough
  (params (Enable :type Boolean) (TargetLimit :type Integer))
  (bind _returnvalue (Math|Integer|SelectInt TargetLimit 0 Enable))
  ;[1 1 1] Event to dynamically enable/disable trace through pawns
  (FortWeapon|SetTraceThroughPawnsLimit _returnvalue)
  (FortWeapon|SetTraceThroughPawns Enable)
  (CallFunction|SetWeaponPierceThroughClientRep Enable _returnvalue))

(event Custom|SetWeaponPierceThrough_ClientRep
  (params (Enable :type Boolean) (TargetLimit :type Integer))
  ;[1 1 1] Event to dynamically enable/disable trace through pawns
  (FortWeapon|SetTraceThroughPawnsLimit (Math|Integer|SelectInt TargetLimit 0 Enable))
  (FortWeapon|SetTraceThroughPawns Enable))

(event Custom|HideWeaponMesh
  ;[1 1 1] DEBUG EVENTS -> ke * <event name>
  (Development|SetHiddeninGame (Variables|Visual|GetWeaponMesh) true false)
  (Development|PrintString "HideWeaponMesh"))

(event Custom|ShowWeaponMesh
  ;[1 1 1] DEBUG EVENTS -> ke * <event name>
  (Development|SetHiddeninGame (Variables|Visual|GetWeaponMesh) false false)
  (Development|PrintString "ShowWeaponMesh"))

(event Custom|HideGunMesh
  ;[1 1 1] DEBUG EVENTS -> ke * <event name>
  (Development|SetHiddeninGame (Variables|Default|GetIronSightsMesh1P) true false)
  (Development|PrintString "HideGunMesh"))

(event Custom|ShowGunMesh
  ;[1 1 1] DEBUG EVENTS -> ke * <event name>
  (Development|SetHiddeninGame (Variables|Default|GetIronSightsMesh1P) false false)
  (Development|PrintString "ShowGunMesh"))

(event Custom|HideWeapon
  ;[1 1 1] DEBUG EVENTS -> ke * <event name>
  (FortWeapon|SetWeaponVisibility false true)
  (Development|PrintString "HideWeapon"))

(event Custom|ShowWeapon
  ;[1 1 1] DEBUG EVENTS -> ke * <event name>
  (FortWeapon|SetWeaponVisibility true true)
  (Development|PrintString "ShowWeapon"))

(fn ActivateOrDeactivateWindParticle
  (params (bNewActive :type Boolean))
  (if (Variables|Default|GetIs%20Wind%20Enabled)
    (Components|Activation|SetActive (Variables|Effects|GetMuzzleWindParticle(Empty)) bNewActive true))
  (return))

(fn ActivateReloadSmokeFX
  (bind _reload_empty (Variables|Default|GetReload(Empty)))
  (Utilities|IsValid _reload_empty
    (:Is%20Valid
      (Components|Activation|Activate _reload_empty false))))

(fn ActivateShellsFX
  (params (Bool :type Boolean))
  (if Bool
    (Components|Activation|Activate (Variables|Effects%28Shells%29|GetShells%28empty%29) true)
    (CallFunction|UpdateShellEmittersFX)))

(fn AddRandomScale
  (if (<= (Math|Random|RandomFloatinRange 0.0 1) 0.8)
    (Transformation|SetRelativeScale3D (Variables|Effects|GetMuzzle(Empty)) (Math|Vector|MakeVector (Math|Random|RandomFloatinRange 0.7 1) (Math|Random|RandomFloatinRange 0.7 1) (Math|Random|RandomFloatinRange 0.7 1)))
    (else
      (Transformation|SetRelativeScale3D (Variables|Effects|GetMuzzle(Empty)) (Math|Vector|MakeVector (Math|Random|RandomFloatinRange 0.5 0.6) (Math|Random|RandomFloatinRange 0.5 0.6) (Math|Random|RandomFloatinRange 0.5 0.6))))))

(fn DeactivateMuzzleFX
  (Components|Activation|Deactivate (Variables|Effects|GetMuzzle(Empty))))

(fn DeactivateReloadSmokeFX
  (bind _reload_empty (Variables|Default|GetReload(Empty)))
  (Utilities|IsValid _reload_empty
    (:Is%20Valid
      (Components|Activation|Deactivate _reload_empty))))

(fn DeactivateShellsFX
  (bind _shells_empty (Variables|Effects%28Shells%29|GetShells%28empty%29))
  (if (Components|Activation|IsActive _shells_empty)
    (Components|Activation|Deactivate _shells_empty)))

(fn HandlePawnAndWeaponVisFor1PTargeting
  (params (IsTargeting :type Boolean))
  (bind _ironsightsmesh1p (Variables|Default|GetIronSightsMesh1P))
  (Utilities|FlowControl|Sequence
    (:then_0
      ;[1 1 1] Hide/show the pawn.
      (bind _asfort_pawn (Utilities|Casting|CastToFortPawn (Game|GetInstigator)))
      (FortPawn|SetPawnVisibilityLocalControllerOnly _asfort_pawn IsTargeting false))
    (:then_1
      ;[1 1 1] Hide/show the viewmodel.
      (Development|SetHiddeninGame _ironsightsmesh1p (not IsTargeting) false)
      (FortWeapon|SetHiddeninGameOtherLocalControllersOnly _ironsightsmesh1p true))
    (:then_2
      ;[1 1 1] Hide/show the 3rd person weapon.
      (FortWeapon|SetHiddeninGameControllerOnly (Variables|Visual|GetWeaponMesh) IsTargeting))))

(fn HideIronSightsMesh
  ;[1 0.8348 0] Hide the viewmodel. I'm not sure why we have to do this but other clients will show the viewmodel when another equips the weapon.
  ; UE-53600, this might be the issue. We'll have to manually check hidden in game on the assault rifles.
  (Development|SetHiddeninGame (Variables|Default|GetIronSightsMesh1P) true false))

(fn HideReticle
  (bind _asfort_player_controller_zone (Utilities|Casting|CastToFortPlayerControllerZone (Game|GetInstigatorController)))
  (HUD|SetHUDElementVisibilitywithContext _asfort_player_controller_zone (Variables|Scope|GetReticleHUDElementTags) true "None" ""))

(fn MuzzleFlashFX
  (params (Persistent%20Fire :type Boolean))
  (bind _returnvalue (Camera|GetIsTargeting))
  (bind _returnvalue_2 (Targeting|UseFirstPersonTargeting))
  (bind _returnvalue_3 (and _returnvalue_2 _returnvalue))
  ;[1 1 1] If local player is using the scope, don't play muzzle effects
  (if (not (Weapon|IsUsingScope))
    (if (not Persistent%20Fire)
      (CallFunction|AddRandomScale))
    (if _returnvalue_3
      ;[1 0.9262 0] For first person targeting, move the muzzle fx to the muzzle location that is set on the weapon (ScopeTargetingMuzzleOffset).
      ; We have to detach first because it will render in a bad position for the first few shots. This unfortunately means that strafing while shooting will have a drifting effect.
      (Transformation|DetachFromComponent (Variables|Effects|GetMuzzle1P(Empty)) "KeepRelative" "KeepRelative" "KeepRelative" true)
      (Transformation|SetWorldLocationAndRotation (Variables|Effects|GetMuzzle1P(Empty)) (Effects|GetMuzzleLocation -1) (Math|Rotator|MakeRotator (.roll (Transformation|GetWorldRotation (Variables|Default|GetIronSightsMesh1P))) (.pitch (Transformation|GetWorldRotation (Variables|Default|GetIronSightsMesh1P))) (+ (.yaw (Transformation|GetWorldRotation (Variables|Default|GetIronSightsMesh1P))) 90)) false false))
    (Components|Activation|Activate (select _returnvalue_3 (Variables|Effects|GetMuzzle1P(Empty)) (Variables|Effects|GetMuzzle(Empty))) true)
    (Variables|Effects|SetLastPlayFXTime (Utilities|Time|GetGameTimeinSeconds)))
  (return))

(fn MuzzlePlayReloadFX
  (params (Selection :type EFortReloadFXState%20Enum :default "ReloadStart"))
  (switch Utilities|FlowControl|Switch|SwitchonEFortReloadFXState Selection
    (:ReloadStart
      ;[1 1 1] Start Reload Smoke
      (if (Variables|Effects|GetUse%20Reload%20Particles)
        (CallFunction|ActivateReloadSmokeFX)))
    (:ReloadEnd
      ;[1 1 1] Stop reload smoke
      (if (Variables|Effects|GetUse%20Reload%20Particles)
        (CallFunction|DeactivateReloadSmokeFX)))
    (:Max_None
      ;[1 1 1] Reload Shells
      (CallFunction|ActivateShellsFX (Variables|Effects%28Shells%29|GetUseShellsOnReload?)))))

(fn SetupShellFX
  (Effects|Components|ParticleSystem|SetTemplate (Variables|Effects%28Shells%29|GetShells%28empty%29) (Variables|Effects%28Shells%29|GetShellsParticleSystemTemplate))
  (Transformation|AttachComponentToComponent (Variables|Effects%28Shells%29|GetShells%28empty%29) (Variables|Visual|GetWeaponMesh) (Variables|Effects%28Shells%29|GetShells%20Socket%20Name) "KeepRelative" "KeepRelative" "KeepRelative" true)
  (if (Variables|Effects%28Shells%29|GetDebugShellsSocket?)
    (if (Transformation|DoesSocketExist (Variables|Visual|GetWeaponMesh) (Variables|Effects%28Shells%29|GetShells%20Socket%20Name))
      (Development|PrintString "***Success! Socket Found!***" true true "(R=0.000000,G=1.000000,B=0.000000,A=1.000000)" 5)
      (else
        (Development|PrintString "***Error! Cannot find socket name to attach shell eject FX to!***" true true "(R=1.000000,G=0.000000,B=0.000000,A=1.000000)" 5))))
  (Effects|Components|ParticleSystem|SetFloatParameter (Variables|Effects%28Shells%29|GetShells%28empty%29) "SpawnRateScale" (Variables|Effects%28Shells%29|GetShells%20Spawn%20Rate%20Scale))
  (Effects|Components|ParticleSystem|SetVectorParameter (Variables|Effects%28Shells%29|GetShells%28empty%29) "ShellsRotationRate" (Variables|Effects%28Shells%29|GetShellsRotationRate))
  (Effects|Components|ParticleSystem|SetVectorParameter (Variables|Effects%28Shells%29|GetShells%28empty%29) "Velocity" (Variables|Effects%28Shells%29|GetShells%20Velocity))
  (Effects|Components|ParticleSystem|SetVectorParameter (Variables|Effects%28Shells%29|GetShells%28empty%29) "Gravity" (Variables|Effects%28Shells%29|GetShells%20Gravity))
  (Effects|Components|ParticleSystem|SetFloatParameter (Variables|Effects%28Shells%29|GetShells%28empty%29) "Lifetime" (Variables|Effects%28Shells%29|GetShells%20Lifetime))
  (Effects|Components|ParticleSystem|SetVectorParameter (Variables|Effects%28Shells%29|GetShells%28empty%29) "Size" (Variables|Effects%28Shells%29|GetShells%20Size))
  (Effects|Components|ParticleSystem|SetVectorParameter (Variables|Effects%28Shells%29|GetShells%28empty%29) "InheritParentVelocity" (Math|Conversions|ToVector(Float) (Variables|Effects%28Shells%29|GetInherit%20Parent%20Velocity)))
  (Effects|Components|ParticleSystem|SetFloatParameter (Variables|Effects%28Shells%29|GetShells%28empty%29) "CylinderRadius" (Variables|Effects%28Shells%29|GetCylindrical%20Radius))
  (Effects|Components|ParticleSystem|SetFloatParameter (Variables|Effects%28Shells%29|GetShells%28empty%29) "CylinderHeight" (Variables|Effects%28Shells%29|GetCylindrical%20Height))
  (Variables|/Script/CascadeRuntime.ParticleSystemComponent|SetCustomTimeDilation (Variables|Effects%28Shells%29|GetShells%20Time%20Dilation) (Variables|Effects%28Shells%29|GetShells%28empty%29))
  (switch Utilities|FlowControl|Switch|SwitchonEn_ShellTypes_01 (Variables|Effects%28Shells%29|GetShellTypeSelect)
    (:NewEnumerator0
      (Variables|Effects%28Shells%29|SetSmallShells true))
    (:NewEnumerator1
      (Variables|Effects%28Shells%29|SetMediumShells true))
    (:NewEnumerator2
      (Variables|Effects%28Shells%29|SetLargeShells true))
    (:NewEnumerator3
      (Variables|Effects%28Shells%29|SetShotgunShells true))
    (:NewEnumerator4
      (Variables|Effects%28Shells%29|SetEnergyShells true))))

(fn SetWpnRarity)

(fn ShowReticle
  (bind _asfort_player_controller_zone (Utilities|Casting|CastToFortPlayerControllerZone (Game|GetInstigatorController)))
  (HUD|ResetHUDElementVisibilitytoSettings _asfort_player_controller_zone (Variables|Scope|GetReticleHUDElementTags)))

(fn UpdateShellEmittersFX
  (Utilities|FlowControl|Sequence
    (:then_0
      ;[1 1 1] Looping Shell Emitters
      (Effects|Components|ParticleSystem|SetEmitterEnable (Variables|Effects%28Shells%29|GetShells%28empty%29) "Small Loop" (and (Variables|Effects%28Shells%29|GetSmallShells) (Effects|ShouldPlayPersistentFireFX)))
      (Effects|Components|ParticleSystem|SetEmitterEnable (Variables|Effects%28Shells%29|GetShells%28empty%29) "Medium Loop" (and (Variables|Effects%28Shells%29|GetMediumShells) (Effects|ShouldPlayPersistentFireFX)))
      (Effects|Components|ParticleSystem|SetEmitterEnable (Variables|Effects%28Shells%29|GetShells%28empty%29) "Large Loop" (and (Variables|Effects%28Shells%29|GetLargeShells) (Effects|ShouldPlayPersistentFireFX)))
      (Effects|Components|ParticleSystem|SetEmitterEnable (Variables|Effects%28Shells%29|GetShells%28empty%29) "Energy Cell Loop" (and (Variables|Effects%28Shells%29|GetEnergyShells) (Effects|ShouldPlayPersistentFireFX)))
      ;[1 1 1] Single Shell Emitters
      (Effects|Components|ParticleSystem|SetEmitterEnable (Variables|Effects%28Shells%29|GetShells%28empty%29) "Small Single" (and (Variables|Effects%28Shells%29|GetSmallShells) (not (Effects|ShouldPlayPersistentFireFX))))
      (Effects|Components|ParticleSystem|SetEmitterEnable (Variables|Effects%28Shells%29|GetShells%28empty%29) "Medium Single" (and (Variables|Effects%28Shells%29|GetMediumShells) (not (Effects|ShouldPlayPersistentFireFX))))
      (Effects|Components|ParticleSystem|SetEmitterEnable (Variables|Effects%28Shells%29|GetShells%28empty%29) "Large Single" (and (Variables|Effects%28Shells%29|GetLargeShells) (not (Effects|ShouldPlayPersistentFireFX))))
      (Effects|Components|ParticleSystem|SetEmitterEnable (Variables|Effects%28Shells%29|GetShells%28empty%29) "Shotgun Single" (and (Variables|Effects%28Shells%29|GetShotgunShells) (not (Effects|ShouldPlayPersistentFireFX))))
      (Effects|Components|ParticleSystem|SetEmitterEnable (Variables|Effects%28Shells%29|GetShells%28empty%29) "Energy Cell Single" (and (Variables|Effects%28Shells%29|GetEnergyShells) (not (Effects|ShouldPlayPersistentFireFX)))))
    (:then_1
      (if (Variables|Effects%28Shells%29|GetUseShellsOnReload?)
        (Effects|Components|ParticleSystem|SetFloatParameter (Variables|Effects%28Shells%29|GetShells%28empty%29) "BurstScale" (Math|Conversions|ToFloat(Integer) (Utilities|ClampMinInteger (CallFunction|GetShellReloadCounter))))
        (CallFunction|SetShellReloadCounter -1)))))

(fn ConstructionScript
  (bind _muzzlewindparticle_empty (Variables|Effects|GetMuzzleWindParticle(Empty)))
  (Transformation|AttachComponentToComponent (Variables|Effects|GetMuzzle(Empty)) (Variables|Visual|GetWeaponMesh) (Effects|GetMuzzleSocketName 0) "KeepRelative" "KeepRelative" "KeepRelative" true)
  (Effects|Components|ParticleSystem|SetTemplate (Variables|Effects|GetMuzzle(Empty)) (Variables|Effects|GetMuzzleParticleSystem))
  (Effects|Components|ParticleSystem|SetTemplate (Variables|Effects|GetMuzzle1P(Empty)) (Variables|Effects|GetMuzzleParticleSystem1P))
  ;[1 1 1] Reload Particles
  (if (Variables|Effects|GetUse%20Reload%20Particles)
    (Utilities|IsValid (Variables|Effects|GetReload_ParticleSystem)
      (:Is%20Valid
        (Effects|Components|ParticleSystem|SetTemplate (Variables|Default|GetReload(Empty)) (Variables|Effects|GetReload_ParticleSystem))
        (Transformation|AttachComponentToComponent (Variables|Default|GetReload(Empty)) (Variables|Visual|GetWeaponMesh) (Variables|Effects|GetReloadSocketName) "KeepRelative" "KeepRelative" "KeepRelative" true))
      (:Is%20Not%20Valid
        _)))
  ;[1 1 1] Wind Particle Setup
  ;; Set Is Wind Enabled
  (bind _returnvalue (Rendering|Material|GetScalarParameterValue "/Game/Packages/Fortress_SharedMaterials/GlobalMaterialParameters/FortniteMaterialParameters.FortniteMaterialParameters" "Water Interaction Enabled" :!class /Script/Engine.KismetMaterialLibrary))
  (bind _output_get (Variables|Default|SetIs%20Wind%20Enabled (> _returnvalue 0.5)))
  (if (not _output_get)
    (Components|DestroyComponent (Variables|Effects|GetMuzzleWindParticle(Empty))))
  (if (Variables|Default|GetIs%20Wind%20Enabled)
    (Effects|Components|ParticleSystem|SetTemplate _muzzlewindparticle_empty (Variables|Effects|GetMuzzleWindParticleSystem))
    (Transformation|AttachComponentToComponent _muzzlewindparticle_empty (Variables|Visual|GetWeaponMesh) (Effects|GetMuzzleSocketName 0) "KeepRelative" "KeepRelative" "KeepRelative" true))
  ;[1 1 1] Shells FX setup
  (if (Math|Boolean|ORBoolean (Variables|Effects%28Shells%29|GetUseShellsOnFire?) (Variables|Effects%28Shells%29|GetUseShellsOnReload?) (Variables|Effects%28Shells%29|GetUseShellsOnPump?))
    (CallFunction|SetupShellFX)))
