(event Custom|BindFireRateChange
  ;[1 1 1] If Auto weapon, Bind Rate of fire change event for fire sound pitch then call it one time to set it initially
  (switch Utilities|FlowControl|Switch|SwitchonEFortWeaponTriggerType (WeaponItemDefinition|GetWeaponDataTriggerType)
    (:Automatic
      (FortWeapon|BindEventtoOnWeaponRateOfFireChanged (delegate Custom|PitchUpOnRateOfFireChange))
      (CallFunction|PitchUponRateOfFireChange (FortWeapon|GetWeaponRateOfFireModifier)))))

(event Custom|ShellEjectionFixOn
  ;[1 1 1] Fix for Shell Ejection Issue on Quality Change
  (EventDispatchers|BindEventtoOnEffectsQualityChanged (Settings|GetFortGameUserSettings) (delegate Custom|Bind%20on%20Effects%20Quality)))

(event Custom|ShellEjectionOff
  ;[1 1 1] Fix for Shell Ejection Issue on Quality Change
  (EventDispatchers|UnbindEventfromOnEffectsQualityChanged (Settings|GetFortGameUserSettings) (delegate Custom|Bind%20on%20Effects%20Quality)))

(event Custom|PitchUpOnRateOfFireChange
  (params (NewRateOfFire :type Float%20%28single-precision%29))
  ;[1 1 1] If Auto weapon, Bind Rate of fire change event for fire sound pitch then call it one time to set it initially
  ;[1 1 1] Pitch on weapon rate of fire change
  (FortWeapon|SetFireSoundPitchMultiplier (Math|Curves|GetFloatValue (Variables|Default|GetCurve_PitchOffset) NewRateOfFire)))

(event Custom|Bind%20on%20Effects%20Quality
  ;[1 1 1] Fix for Shell Ejection Issue on Quality Change
  (Utilities|FlowControl|Delay 0.1) ; REQUIRED to make work!
  (CallFunction|UpdateShellEmittersFX)
  (CallFunction|DeactivateShellsFX))

(event EventBeginPlay
  (Rendering|SetVisibility (Variables|Visual|GetWeaponMesh) false false) ; these are always invisible
  (bind _asfort_player_pawn (Utilities|Casting|CastToFortPlayerPawn (Game|GetInstigator)))
  (Variables|Default|SetPlayerPawn _asfort_player_pawn)
  ;[0 0 0] Send wind particle reference to the wind manager
  (if (Variables|Default|GetIs%20Wind%20Enabled)
    (Wind|AddWindParticleSystemComponent (Fortnite|GetWindManager) (Variables|Effects|GetMuzzleWindParticleSpawned))))

(event Custom|ShellsON_(onPump)
  ;[0.3841 0.6674 1] Shells VFX (on Pump)
  (CallFunction|ActivateShellsFX (and (Variables|Effects%28Shells%29|GetUseShellsOnPump?) (> (Rendering|GetRenderingDetailMode) 0))))

(event EventOnPlayWeaponFireFX
  (params (bPersistentFire :type Boolean) (bSecondaryFire :type Boolean))
  ;[0.4433 0 1] Muzzle Flash VFX
  (Utilities|FlowControl|Sequence
    (:then_0
      (CallFunction|MuzzleFlashFX bPersistentFire))
    (:then_1
      (CallFunction|ActivateorDeactivateWindParticle true))
    (:then_2
      (if (and (not (Weapon|IsUsingScopeorFirstPersonCamera)) (> (Rendering|GetRenderingDetailMode) 0))
        (CallFunction|ActivateShellsFX (Variables|Effects%28Shells%29|GetUseShellsOnFire?))))
    (:then_3
      (FortWeapon|PlayForceFeedback false bPersistentFire))))

(event EventOnStopWeaponFireFX
  (bind _returnvalue (Utilities|Time|GetGameTimeinSeconds))
  ;[0.4433 0 1] Muzzle Flash VFX
  ;[1 1 1] Enforce a short minimum time to play looping FX
  (if (not (>= (- _returnvalue (Variables|Effects|GetLastPlayFXTime)) (Variables|Effects|GetMinPlayFXTime)))
    (Utilities|FlowControl|RetriggerableDelay (- (Variables|Effects|GetMinPlayFXTime) (- _returnvalue (Variables|Effects|GetLastPlayFXTime)))))
  (CallFunction|ActivateorDeactivateWindParticle false)
  (if (Effects|ShouldPlayPersistentFireFX)
    (CallFunction|DeactivateMuzzleFX))
  (Effects|DeactivateBeamFX)
  (CallFunction|DeactivateShellsFX)
  (FortWeapon|StopForceFeedback))

(event EventOnInitCosmeticAlterations
  (params (CosmeticMod :type Fort%20Cosmetic%20Modification%20Structure))
  (bind (_cosmeticmaterial _ambientparticlesystem _ambientniagarasystem _muzzleparticlesystem _muzzleniagarasystem _reloadparticlesystem _reloadniagarasystem _beamparticlesystem _beamniagarasystem _tracertemplate _bmodifycolor _coloralteration _colorparametername _bmodifydecalcolour _decalcolouralterationstart _decalcolouralterationend _bmodifyshellcolour _shellcolouralteration) (Utilities|Struct|BreakFortCosmeticModification CosmeticMod))
  (bind (_r _g _b) (Utilities|Struct|BreakLinearColor _shellcolouralteration))
  (bind _returnvalue (CallFunction|GetBeamPSC))
  (bind _muzzle_empty (Variables|Effects|GetMuzzle(Empty)))
  (bind _reload_empty (Variables|Default|GetReload(Empty)))
  (bind _shells_empty (Variables|Effects%28Shells%29|GetShells%28empty%29))
  ;[0.06096 0.8755 1] Setup Alteration FX
  ;; Alteration Logic
  (Networking|CanExecuteCosmeticEvents
    (:True
      (Utilities|FlowControl|Sequence
        (:then_0
          (if (not (Asset|IsAssetNull _ambientniagarasystem))
            (bind _object (Utilities|AsyncLoadAsset _ambientniagarasystem)
              (:Completed
                (bind _asniagara_system (Utilities|Casting|CastToNiagaraSystem _object))
                (bind _alteration_ambient_ps (Variables|Effects%28Alterations%29|GetAlteration%20Ambient%20PS)
                  (:then
                    (Components|DestroyComponent _alteration_ambient_ps))
                  (:else
                    _))
                (bind _niagara_component_system_reference (Utilities|ElementalWeaponSetupFX (Variables|Visual|GetWeaponMesh) _asniagara_system "FX_Idle_Start" "FX_Idle_End" false)
                  (:Out
                    (bind _output_get (Variables|Effects%28Alterations%29|SetAlteration%20Ambient%20PS _niagara_component_system_reference))
                    (LOD|SetMaxDrawDistance _output_get (* (Variables|EffectsDrawDistances|GetAlteration%20Ambient%20PS%20Max%20Draw%20Distance) (Utilities|Reciprocal (Fortnite|GetNumericCVar "r.ViewDistanceScale"))))))))))
        (:then_1
          (if (not (Asset|IsAssetNull _muzzleparticlesystem))
            (bind _object_2 (Utilities|AsyncLoadAsset _muzzleparticlesystem)
              (:Completed
                (bind _asparticle_system (Utilities|Casting|CastToParticleSystem _object_2))
                (Effects|Components|ParticleSystem|SetTemplate _muzzle_empty _asparticle_system)
                (LOD|SetMaxDrawDistance _muzzle_empty (* (Variables|EffectsDrawDistances|GetMuzzle%20PS%20Max%20Draw%20Distance) (Utilities|Reciprocal (Fortnite|GetNumericCVar "r.ViewDistanceScale"))))))))
        (:then_2
          (if (and (Effects|GetUseBeamParticles) (not (Asset|IsAssetNull _beamparticlesystem)))
            (bind _object_3 (Utilities|AsyncLoadAsset _beamparticlesystem)
              (:Completed
                (bind _asparticle_system_2 (Utilities|Casting|CastToParticleSystem _object_3))
                (bind _returnvalue_2 (Effects|SetBeamParticleSystem _asparticle_system_2))
                (Effects|Components|ParticleSystem|SetTemplate _returnvalue _returnvalue_2)
                (LOD|SetMaxDrawDistance _returnvalue (* (Variables|EffectsDrawDistances|GetBeam%20PS%20Max%20Draw%20Distance) (Utilities|Reciprocal (Fortnite|GetNumericCVar "r.ViewDistanceScale"))))))))
        (:then_3
          (if (not (Asset|IsAssetNull _reloadparticlesystem))
            (bind _object_4 (Utilities|AsyncLoadAsset _reloadparticlesystem)
              (:Completed
                (bind _asparticle_system_3 (Utilities|Casting|CastToParticleSystem _object_4))
                (Effects|Components|ParticleSystem|SetTemplate _reload_empty _asparticle_system_3)
                (LOD|SetMaxDrawDistance _reload_empty (* (Variables|EffectsDrawDistances|GetReload%20PS%20Max%20Draw%20Distance) (Utilities|Reciprocal (Fortnite|GetNumericCVar "r.ViewDistanceScale"))))))))
        (:then_4
          (if _bmodifydecalcolour
            (Decals|SetDecalColorStart _decalcolouralterationstart)
            (Decals|SetDecalColorEnd _decalcolouralterationend)))
        (:then_5
          (if _bmodifyshellcolour
            (Effects|Components|ParticleSystem|SetFloatParameter (Variables|Effects%28Shells%29|GetShells%28empty%29) "Red" _r)
            (Effects|Components|ParticleSystem|SetFloatParameter (Variables|Effects%28Shells%29|GetShells%28empty%29) "Green" _g)
            (Effects|Components|ParticleSystem|SetFloatParameter _shells_empty "Blue" _b))
          (LOD|SetMaxDrawDistance _shells_empty (* (Variables|EffectsDrawDistances|GetShells%20PS%20Max%20Draw%20Distance) (Utilities|Reciprocal (Fortnite|GetNumericCVar "r.ViewDistanceScale")))))
        (:then_6
          (bind _asplayer_pawn_generic_parent (Utilities|Casting|CastToPlayerPawn_Generic_Parent (Game|GetInstigator self)))
          (Class|PlayerPawnGenericParent|ReinitializeWeaponMaterials _asplayer_pawn_generic_parent))))))

(event EventOnPlayReloadFX
  (params (ReloadStage :type EFortReloadFXState%20Enum :default "ReloadStart"))
  ;[0.3518 0.0375 0.63] Use Reload Smoke Trails
  (CallFunction|MuzzlePlayReloadFX ReloadStage)
  (FortWeapon|PlayForceFeedback true))

(event EventOnInitAlteration
  (params (NewAlteration :type Fort%20Alteration%20Item%20Definition%20Object%20Reference)))

(event FortWeapon|EventOnEquippedAndAttached
  (bind _scopemesh1p (Variables|Components|GetScopeMesh1P))
  ;[1 1 1] On weapon equip
  ;;[1 1 1] Check if in athena, if not set up fire rate bind.
  (if (not (FortGlobals|IsinAthena))
    (Networking|SwitchHasAuthority
      (:Authority
        _)
      (:Remote
        (CallFunction|BindFireRateChange))))
  (CallFunction|ShowReticle)
  (CallFunction|ShellEjectionFixOn)
  (Utilities|FlowControl|Sequence
    (:then_0
      ;[1 1 1] Init Scope Variables
      (bind _asfort_pawn (Utilities|Casting|CastToFortPawn (Game|GetInstigator)))
      (Utilities|IsValid (Pawn|GetLocalViewingPlayerController _asfort_pawn)
        (:Is%20Valid
          (Networking|CanExecuteCosmeticEvents
            (:True
              (if (Scope|UseScopeTargeting)
                (Utilities|IsValid _scopemesh1p
                  (:Is%20Valid
                    (Rendering|Material|SetVectorParameterValue "/Game/Packages/Fortress_SharedMaterials/GlobalMaterialParameters/FortniteMaterialParameters.FortniteMaterialParameters" "TopLeftScope" (Math|Conversions|ToLinearColor(Vector) (Transformation|GetSocketLocation _scopemesh1p "TopLeftScope")) :!class /Script/Engine.KismetMaterialLibrary)
                    (Rendering|Material|SetVectorParameterValue "/Game/Packages/Fortress_SharedMaterials/GlobalMaterialParameters/FortniteMaterialParameters.FortniteMaterialParameters" "BottomRightScope" (Math|Conversions|ToLinearColor(Vector) (Transformation|GetSocketLocation _scopemesh1p "BottomRightScope")) :!class /Script/Engine.KismetMaterialLibrary)))))))))
    (:then_1
      (Utilities|FlowControl|Delay 0.1) ; This is a hack fix
      (CallFunction|SetActiveAlterationIdleParticles true))))

(event Custom|InitializeScopeVariables
  (bind _scopemesh1p (Variables|Components|GetScopeMesh1P))
  ;[1 1 1] Init Scope Variables
  (bind _asfort_pawn (Utilities|Casting|CastToFortPawn (Game|GetInstigator)))
  (Utilities|IsValid (Pawn|GetLocalViewingPlayerController _asfort_pawn)
    (:Is%20Valid
      (Networking|CanExecuteCosmeticEvents
        (:True
          (if (Scope|UseScopeTargeting)
            (Utilities|IsValid _scopemesh1p
              (:Is%20Valid
                (Rendering|Material|SetVectorParameterValue "/Game/Packages/Fortress_SharedMaterials/GlobalMaterialParameters/FortniteMaterialParameters.FortniteMaterialParameters" "TopLeftScope" (Math|Conversions|ToLinearColor(Vector) (Transformation|GetSocketLocation _scopemesh1p "TopLeftScope")) :!class /Script/Engine.KismetMaterialLibrary)
                (Rendering|Material|SetVectorParameterValue "/Game/Packages/Fortress_SharedMaterials/GlobalMaterialParameters/FortniteMaterialParameters.FortniteMaterialParameters" "BottomRightScope" (Math|Conversions|ToLinearColor(Vector) (Transformation|GetSocketLocation _scopemesh1p "BottomRightScope")) :!class /Script/Engine.KismetMaterialLibrary)))))))))

(event Custom|UnhideThirdPersonStuff
  ;[1 1 1] Scoped weapon targeting downsights
  ;;[1 1 1] Hiding/showing 3p meshes, etc
  (if (not (Weapon|IsUsingFirstPersonCamera))
    (bind _asfort_pawn (Utilities|Casting|CastToFortPawn (Game|GetInstigator)))
    (FortPawn|SetPawnVisibilityLocalControllerOnly _asfort_pawn false false)
    (FortWeapon|SetHiddeninGameControllerOnly (Variables|Visual|GetWeaponMesh)))
  (CallFunction|ShowReticle)
  (bind _asfort_player_controller_zone (Utilities|Casting|CastToFortPlayerControllerZone (Game|GetInstigatorController)))
  (Class|FortPlayerController|FixupInteractionWidgetsonUnzoom _asfort_player_controller_zone))

(event Custom|HideFirstPersonStuff
  (bind _scopemesh1p (Variables|Components|GetScopeMesh1P))
  ;[1 1 1] Scoped weapon targeting downsights
  ;;[1 1 1] Hiding/showing 1p meshes, etc
  (Rendering|SetVisibility _scopemesh1p false false)
  (Development|SetHiddeninGame _scopemesh1p true false)
  (bind _downscopepostprocessspawned (Variables|Scope|GetDownScopePostProcessSpawned)
    (:then
      (Variables|/Script/Engine.PostProcessComponent|SetbEnabled false _downscopepostprocessspawned))
    (:else
      _))
  (bind _downscopepostprocessspawned2 (Variables|Scope|GetDownScopePostProcessSpawned2)
    (:then
      (Variables|/Script/Engine.PostProcessComponent|SetbEnabled false _downscopepostprocessspawned2))
    (:else
      _))
  ;;;[1 1 1] Custom Depth
  (if (Variables|Scope|GetScope%20-%20Render%20Enemies%20To%20Custom%20Depth%20Buffer)
    (CallFunction|UpdateEnemyCustomDepths false 0)))

(event EventOnSetTargeting
  (params (bNewIsTargeting :type Boolean))
  ;[1 1 1] Scoped weapon targeting downsights
  (Networking|CanExecuteCosmeticEvents
    (:True
      (bind _asfort_pawn (Utilities|Casting|CastToFortPawn (Game|GetInstigator)))
      (Utilities|IsValid (Pawn|GetLocalViewingPlayerController _asfort_pawn)
        (:Is%20Valid
          (Default|CallOnAimDownSightsChanged bNewIsTargeting)
          (if (Scope|UseScopeTargeting)
            (if bNewIsTargeting
              (CallFunction|InitializeScopeVariables)
              (Utilities|FlowControl|Sequence
                (:then_0
                  (Utilities|FlowControl|Delay 0.1)
                  (if (Camera|GetIsTargeting)
                    (Utilities|FlowControl|Sequence
                      (:then_0
                        (CallFunction|UnhideFirstPersonStuffPart1)
                        (CallFunction|PlayScopePP)
                        (CallFunction|SetActiveAlterationIdleParticles false))
                      (:then_1
                        (Utilities|FlowControl|DoOnce false
                          (:Completed
                            (CallFunction|StartScopedAudio)))))))
                (:then_1
                  (Utilities|FlowControl|Delay 0.2)
                  (if (Camera|GetIsTargeting)
                    (CallFunction|HideThirdPersonStuff)
                    (CallFunction|UnhideFirstPersonStuffPart2 2))))
              (else
                (Utilities|FlowControl|Sequence
                  (:then_0
                    (CallFunction|StopScopedAudio)
                    (CallFunction|UnhideThirdPersonStuff))
                  (:then_1
                    (CallFunction|ForceScopeBackImmediatly))
                  (:then_2
                    (CallFunction|ResetDoonceScopeSound)
                    (CallFunction|SetActiveAlterationIdleParticles true)))))
            (else
              (CallFunction|ShowReticle))))))))

(event Custom|ResetDoonceScopeSound
  ;[1 1 1] Scoped weapon targeting downsights
  (Utilities|FlowControl|DoOnce false
    (:Completed
      (CallFunction|StartScopedAudio))))

(event Custom|HideThirdPersonStuff
  ;[1 1 1] Scoped weapon targeting downsights
  ;;[1 1 1] Hiding/showing 3p meshes, etc
  (bind _asfort_pawn (Utilities|Casting|CastToFortPawn (Game|GetInstigator)))
  (FortPawn|SetPawnVisibilityLocalControllerOnly _asfort_pawn true false)
  (FortWeapon|SetWeaponVisibilityControllerOnly true true)
  (FortWeapon|SetHiddeninGameControllerOnly (Variables|Visual|GetWeaponMesh) true))

(event Custom|UnhideFirstPersonStuffPart2
  (params (Which%20Call :type Integer))
  (bind _scopemesh1p (Variables|Components|GetScopeMesh1P))
  ;[1 1 1] Scoped weapon targeting downsights
  ;;[1 1 1] Hiding/showing 1p meshes, etc
  (Rendering|SetVisibility _scopemesh1p true false)
  (Development|SetHiddeninGame _scopemesh1p false false)
  (FortWeapon|SetHiddeninGameOtherLocalControllersOnly _scopemesh1p true)
  ;;;[1 1 1] Custom Depth
  (if (Variables|Scope|GetScope%20-%20Render%20Enemies%20To%20Custom%20Depth%20Buffer)
    (CallFunction|UpdateEnemyCustomDepths true (Variables|Default|GetStencilBufferValue))))

(event FortWeapon|EventOnLocalUnEquip
  ;[1 1 1] Scoped weapon targeting downsights
  (CallFunction|SetActiveAlterationIdleParticles false)
  (bind _asfort_pawn (Utilities|Casting|CastToFortPawn (Game|GetInstigator)))
  (Utilities|IsValid (Pawn|GetLocalViewingPlayerController _asfort_pawn)
    (:Is%20Valid
      (if (Scope|UseScopeTargeting)
        (Utilities|FlowControl|Sequence
          (:then_0
            (CallFunction|StopScopedAudio)
            (CallFunction|UnhideThirdPersonStuff))
          (:then_1
            (CallFunction|ForceScopeBackImmediatly))
          (:then_2
            (CallFunction|ResetDoonceScopeSound)
            (CallFunction|SetActiveAlterationIdleParticles true)))
        (else
          (CallFunction|ShowReticle))))))

(event Custom|UnhideFirstPersonStuffPart1
  ;[1 1 1] Scoped weapon targeting downsights
  ;;[1 1 1] Hiding/showing 1p meshes, etc
  (bind _downscopepostprocessspawned (Variables|Scope|GetDownScopePostProcessSpawned)
    (:then
      (Variables|/Script/Engine.PostProcessComponent|SetbEnabled true _downscopepostprocessspawned))
    (:else
      _))
  (bind _downscopepostprocessspawned2 (Variables|Scope|GetDownScopePostProcessSpawned2)
    (:then
      (Variables|/Script/Engine.PostProcessComponent|SetbEnabled true _downscopepostprocessspawned2))
    (:else)))

(event Custom|PlayScopePP
  ;[1 1 1] Scoped weapon targeting downsights
  ;;[1 1 1] Scope Post Process
  (bind (_direction _downsightpostprocessamount) (Timeline|AnimateScopePostProcess PlayFromStart)
    (:Update
      (CallFunction|SetPostProcessParams _downsightpostprocessamount))
    (:Finished
      (if (Utilities|Enum|Equal(Enum) _direction "Backward")
        (if (not (Camera|GetIsTargeting))
          (CallFunction|HideFirstPersonStuff))))))

(event Custom|ReverseScopePP
  ;[1 1 1] Scoped weapon targeting downsights
  ;;[1 1 1] Scope Post Process
  (bind _asfort_player_pawn (Utilities|Casting|CastToFortPlayerPawn (Game|GetInstigator)))
  (if (Vehicle|IsinVehicle _asfort_player_pawn)
    (if (not (Camera|GetIsTargeting))
      (CallFunction|HideFirstPersonStuff))
    (else
      (Timeline|AnimateScopePostProcess Reverse))))

(event Targeting|EventAbortScopeFX
  ;[1 1 1] Direct control of scoping, useful when switching between spectated characters
  (Utilities|FlowControl|Sequence
    (:then_0
      (CallFunction|SetPostProcessParams 0.0))
    (:then_1
      (CallFunction|HideFirstPersonStuff))
    (:then_2
      (CallFunction|UnhideThirdPersonStuff))))

(event Targeting|EventForceScopeFX
  ;[1 1 1] Direct control of scoping, useful when switching between spectated characters
  (Utilities|FlowControl|Sequence
    (:then_0
      (CallFunction|SetPostProcessParams 1.0))
    (:then_1
      (CallFunction|UnhideFirstPersonStuffPart1)
      (CallFunction|UnhideFirstPersonStuffPart2 1))
    (:then_2
      (CallFunction|HideThirdPersonStuff))))

(event Custom|ForceScopeBackImmediatly
  (bind _asfort_player_pawn (Utilities|Casting|CastToFortPlayerPawn (Game|GetInstigator)))
  (if (not (Vehicle|IsinVehicle _asfort_player_pawn))
    (CallFunction|SetPostProcessParams 0.0))
  (if (not (Camera|GetIsTargeting))
    (CallFunction|HideFirstPersonStuff)))

(event EventOnWeaponVisibilityChanged
  (params (bVisible :type Boolean) (bSetForLocalControllerOnly :type Boolean))
  (bind _alteration_ambient_ps (Variables|Effects%28Alterations%29|GetAlteration%20Ambient%20PS))
  ;[1 1 1] EVENT -> OnWeaponVisibilityChanged
  (Utilities|IsValid (Variables|Effects%28Alterations%29|GetAlteration%20Ambient%20PS)
    (:Is%20Valid
      (if bVisible
        (Components|Activation|Activate _alteration_ambient_ps false)
        (else
          (Components|Activation|Deactivate _alteration_ambient_ps))))))

(event Custom|Update%20Enemy%20Custom%20Depths
  (params (Enable%20Or%20Disable :type Boolean :default true) (StencilBufferValue :type Integer :default 3))
  ;[1 1 1] Custom Depth
  (bind _outactors (Actor|GetAllActorsOfClass "/Script/FortniteGame.FortAIPawn"))
  (bind _output_get (Variables|Scope|SetArray%20Of%20Active%20Enemy%20AI _outactors))
  (for _array_element _output_get
    (Utilities|IsValid (Variables|/Script/Engine.Character|GetMesh _array_element)
      (:Is%20Valid
        (Rendering|SetRenderCustomDepth (Variables|/Script/Engine.Character|GetMesh _array_element) Enable%20Or%20Disable)
        (Rendering|SetCustomDepthStencilValue (Variables|/Script/Engine.Character|GetMesh _array_element) StencilBufferValue)))))

(event EventOnEquippedWeaponDestory
  (bind (_tintcolor _drawas _tiling _mirroring _imagesize _margin _resourceobject) (Utilities|Struct|BreakSlateBrush (Item|GetWeaponDataSmallPreviewImageBrush)))
  ;[0.725 0.6854 0.4277] Show fx for weapon destruction on depleted durability
  (if (Variables|Effects|GetUseDestroyEffect)
    (Effects|SpawnEmitteratLocation (Variables|Effects|GetWeaponDurabilityDestroyEffect) (Transformation|GetSocketLocation (Variables|Visual|GetWeaponMesh) "Handle"))
    (bind _asfort_pawn (Utilities|Casting|CastToFortPawn (Game|GetInstigator)))
    (bind _returnvalue (Effects|SpawnEmitterAttached :EmitterTemplate (Variables|Effects|GetWeaponDurabilityDestroyEffectIcon) :AttachToComponent (Variables|/Script/Engine.Character|GetMesh _asfort_pawn) :AttachPointName "Root_Effect" :Location "0, 0,130.000000" :Scale "1.000000,1.000000,1.000000" :LocationType "SnapToTarget" :bAutoDestroy true :PoolingMethod "None" :bAutoActivate true))
    (bind _returnvalue_2 (Rendering|Material|CreateDynamicMaterialInstance _returnvalue 0 "/Game/Effects/WeaponDestroy/M_WeaponDestroyExclamation.M_WeaponDestroyExclamation" "None" :!class /Script/Engine.PrimitiveComponent))
    (Rendering|Material|SetMaterial _returnvalue 0 _returnvalue_2)
    (bind _astexture_2d (Utilities|Casting|CastToTexture2D _resourceobject)
      (:then
        _)
      (:CastFailed
        _))
    (Rendering|Material|SetTextureParameterValue _returnvalue_2 "IconTex" _astexture_2d)))

(event Custom|SetWeaponPierceThrough
  (params (Enable :type Boolean) (TargetLimit :type Integer))
  (bind _returnvalue (Math|Integer|SelectInt TargetLimit 0 Enable))
  ;[1 1 1] Event to dynamically enable/disable trace through pawns
  (FortWeapon|SetTraceThroughPawnsLimit _returnvalue)
  (FortWeapon|SetTraceThroughPawns Enable)
  (CallFunction|SetWeaponPierceThroughClientRep Enable _returnvalue))

(event Custom|SetWeaponPierceThrough_ClientRep
  (params (Enable :type Boolean) (TargetLimit :type Integer))
  ;[1 1 1] Event to dynamically enable/disable trace through pawns
  (FortWeapon|SetTraceThroughPawnsLimit (Math|Integer|SelectInt TargetLimit 0 Enable))
  (FortWeapon|SetTraceThroughPawns Enable))

(event Custom|HideWeaponMesh
  ;[1 1 1] DEBUG EVENTS -> ke * <event name>
  (Development|SetHiddeninGame (Variables|Visual|GetWeaponMesh) true false)
  (Development|PrintString "HideWeaponMesh"))

(event Custom|ShowWeaponMesh
  ;[1 1 1] DEBUG EVENTS -> ke * <event name>
  (Development|SetHiddeninGame (Variables|Visual|GetWeaponMesh) false false)
  (Development|PrintString "ShowWeaponMesh"))

(event Custom|HideWeapon
  ;[1 1 1] DEBUG EVENTS -> ke * <event name>
  (FortWeapon|SetWeaponVisibility false true)
  (Development|PrintString "HideWeapon"))

(event Custom|ShowWeapon
  ;[1 1 1] DEBUG EVENTS -> ke * <event name>
  (FortWeapon|SetWeaponVisibility true true)
  (Development|PrintString "ShowWeapon"))

(fn ActivateOrDeactivateWindParticle
  (params (bNewActive :type Boolean))
  (if (Variables|Default|GetIs%20Wind%20Enabled)
    (Components|Activation|SetActive (Variables|Effects|GetMuzzleWindParticleSpawned) bNewActive true))
  (return))

(fn ActivateReloadSmokeFX
  (bind _reload_empty (Variables|Default|GetReload(Empty)))
  (Utilities|IsValid _reload_empty
    (:Is%20Valid
      (Components|Activation|Activate _reload_empty false))))

(fn ActivateShellsFX
  (params (Bool :type Boolean))
  (bind _shells_empty (Variables|Effects%28Shells%29|GetShells%28empty%29))
  (bind _shells_spawn_rate_scale (Variables|Effects%28Shells%29|GetShells%20Spawn%20Rate%20Scale))
  (if Bool
    (Components|Activation|Activate (Variables|Effects%28Shells%29|GetShells%28empty%29) true)
    (Effects|Components|ParticleSystem|SetFloatParameter _shells_empty "SpawnRateScale" _shells_spawn_rate_scale)
    (Effects|Components|ParticleSystem|SetFloatParameter _shells_empty "BurstScale" _shells_spawn_rate_scale)
    (CallFunction|UpdateShellEmittersFX)))

(fn AddRandomScale
  (if (<= (Math|Random|RandomFloatinRange 0.0 1) 0.8)
    (Transformation|SetRelativeScale3D (Variables|Effects|GetMuzzle(Empty)) (Math|Vector|MakeVector (Math|Random|RandomFloatinRange 0.7 1) (Math|Random|RandomFloatinRange 0.7 1) (Math|Random|RandomFloatinRange 0.7 1)))
    (else
      (Transformation|SetRelativeScale3D (Variables|Effects|GetMuzzle(Empty)) (Math|Vector|MakeVector (Math|Random|RandomFloatinRange 0.5 0.6) (Math|Random|RandomFloatinRange 0.5 0.6) (Math|Random|RandomFloatinRange 0.5 0.6))))))

(fn DeactivateMuzzleFX
  (Components|Activation|Deactivate (Variables|Effects|GetMuzzle(Empty))))

(fn DeactivateReloadSmokeFX
  (bind _reload_empty (Variables|Default|GetReload(Empty)))
  (Utilities|IsValid _reload_empty
    (:Is%20Valid
      (Components|Activation|Deactivate _reload_empty))))

(fn DeactivateShellsFX
  (bind _shells_empty (Variables|Effects%28Shells%29|GetShells%28empty%29))
  (if (Components|Activation|IsActive _shells_empty)
    (Effects|Components|ParticleSystem|SetFloatParameter _shells_empty "SpawnRateScale" 0.0)
    (Effects|Components|ParticleSystem|SetFloatParameter _shells_empty "BurstScale" 0.0)
    (Components|Activation|Deactivate _shells_empty)))

(fn HideReticle
  (bind _asfort_player_controller_zone (Utilities|Casting|CastToFortPlayerControllerZone (Game|GetInstigatorController)))
  (HUD|SetHUDElementVisibilitywithContext _asfort_player_controller_zone (Variables|Scope|GetReticleHUDElementTags) true "None" ""))

(fn MuzzleFlashFX
  (params (Persistent%20Fire :type Boolean))
  ;[1 1 1] If local player is using the scope, don't play muzzle effects
  (if (not (Weapon|IsUsingScope))
    (if (not Persistent%20Fire)
      (CallFunction|AddRandomScale))
    (Components|Activation|Activate (Variables|Effects|GetMuzzle(Empty)) true)
    (Variables|Effects|SetLastPlayFXTime (Utilities|Time|GetGameTimeinSeconds)))
  (return))

(fn MuzzlePlayReloadFX
  (params (Selection :type EFortReloadFXState%20Enum :default "ReloadStart"))
  (switch Utilities|FlowControl|Switch|SwitchonEFortReloadFXState Selection
    (:ReloadStart
      ;[1 1 1] Start Reload Smoke
      (if (Variables|Effects|GetUse%20Reload%20Particles)
        (CallFunction|ActivateReloadSmokeFX)))
    (:ReloadEnd
      ;[1 1 1] Stop reload smoke
      (if (Variables|Effects|GetUse%20Reload%20Particles)
        (CallFunction|DeactivateReloadSmokeFX)))
    (:Max_None
      ;[1 1 1] Reload Shells
      (CallFunction|ActivateShellsFX (Variables|Effects%28Shells%29|GetUseShellsOnReload?)))))

(fn SetActiveAlterationIdleParticles
  (params (Active :type Boolean))
  (Networking|CanExecuteCosmeticEvents
    (:True
      (bind _alteration_ambient_ps (Variables|Effects%28Alterations%29|GetAlteration%20Ambient%20PS)
        (:then
          (if Active
            (Development|SetHiddeninGame _alteration_ambient_ps false false)
            (Components|Activation|Activate _alteration_ambient_ps false)
            (else
              (Development|SetHiddeninGame _alteration_ambient_ps true false)
              (Components|Activation|Deactivate _alteration_ambient_ps)))
          (return))
        (:else)))))

(fn SetPostProcessParams
  (params (InputPin :type Float%20%28double-precision%29))
  (Variables|Scope|SetScopePostProcessBlendWeight InputPin))

(fn SetupShellFX
  (Effects|Components|ParticleSystem|SetTemplate (Variables|Effects%28Shells%29|GetShells%28empty%29) (Variables|Effects%28Shells%29|GetShellsParticleSystemTemplate))
  (Transformation|AttachComponentToComponent (Variables|Effects%28Shells%29|GetShells%28empty%29) (Variables|Visual|GetWeaponMesh) (Variables|Effects%28Shells%29|GetShells%20Socket%20Name) "KeepRelative" "KeepRelative" "KeepRelative" true)
  (if (Variables|Effects%28Shells%29|GetDebugShellsSocket?)
    (if (Transformation|DoesSocketExist (Variables|Visual|GetWeaponMesh) (Variables|Effects%28Shells%29|GetShells%20Socket%20Name))
      (Development|PrintString "***Success! Socket Found!***" true true "(R=0.000000,G=1.000000,B=0.000000,A=1.000000)" 5)
      (else
        (Development|PrintString "***Error! Cannot find socket name to attach shell eject FX to!***" true true "(R=1.000000,G=0.000000,B=0.000000,A=1.000000)" 5))))
  (Effects|Components|ParticleSystem|SetVectorParameter (Variables|Effects%28Shells%29|GetShells%28empty%29) "ShellsRotationRate" (Variables|Effects%28Shells%29|GetShellsRotationRate))
  (Effects|Components|ParticleSystem|SetVectorParameter (Variables|Effects%28Shells%29|GetShells%28empty%29) "Velocity" (Variables|Effects%28Shells%29|GetShells%20Velocity))
  (Effects|Components|ParticleSystem|SetVectorParameter (Variables|Effects%28Shells%29|GetShells%28empty%29) "Gravity" (Variables|Effects%28Shells%29|GetShells%20Gravity))
  (Effects|Components|ParticleSystem|SetFloatParameter (Variables|Effects%28Shells%29|GetShells%28empty%29) "Lifetime" (Variables|Effects%28Shells%29|GetShells%20Lifetime))
  (Effects|Components|ParticleSystem|SetVectorParameter (Variables|Effects%28Shells%29|GetShells%28empty%29) "Size" (Variables|Effects%28Shells%29|GetShells%20Size))
  (Effects|Components|ParticleSystem|SetVectorParameter (Variables|Effects%28Shells%29|GetShells%28empty%29) "InheritParentVelocity" (Math|Conversions|ToVector(Float) (Variables|Effects%28Shells%29|GetInherit%20Parent%20Velocity)))
  (Effects|Components|ParticleSystem|SetFloatParameter (Variables|Effects%28Shells%29|GetShells%28empty%29) "CylinderRadius" (Variables|Effects%28Shells%29|GetCylindrical%20Radius))
  (Effects|Components|ParticleSystem|SetFloatParameter (Variables|Effects%28Shells%29|GetShells%28empty%29) "CylinderHeight" (Variables|Effects%28Shells%29|GetCylindrical%20Height))
  (Variables|/Script/CascadeRuntime.ParticleSystemComponent|SetCustomTimeDilation (Variables|Effects%28Shells%29|GetShells%20Time%20Dilation) (Variables|Effects%28Shells%29|GetShells%28empty%29))
  (switch Utilities|FlowControl|Switch|SwitchonEn_ShellTypes_01 (Variables|Effects%28Shells%29|GetShellTypeSelect)
    (:NewEnumerator0
      (Variables|Effects%28Shells%29|SetSmallShells true))
    (:NewEnumerator1
      (Variables|Effects%28Shells%29|SetMediumShells true))
    (:NewEnumerator2
      (Variables|Effects%28Shells%29|SetLargeShells true))
    (:NewEnumerator3
      (Variables|Effects%28Shells%29|SetShotgunShells true))
    (:NewEnumerator4
      (Variables|Effects%28Shells%29|SetEnergyShells true))))

(fn SetWpnRarity)

(fn ShowReticle
  (bind _asfort_player_controller_zone (Utilities|Casting|CastToFortPlayerControllerZone (Game|GetInstigatorController)))
  (HUD|ResetHUDElementVisibilitytoSettings _asfort_player_controller_zone (Variables|Scope|GetReticleHUDElementTags)))

(fn StartScopedAudio
  (Utilities|FlowControl|Sequence
    (:then_0
      (bind _sound_scopezoomin (Variables|Sounds|GetSound_ScopeZoomIn)
        (:then
          (bind _returnvalue (Audio|SpawnSoundAttached _sound_scopezoomin (Variables|Visual|GetWeaponMesh)))
          (Variables|Default|SetScopeZoomInComp _returnvalue))
        (:else)))
    (:then_1
      (bind _sound_scopedinloop (Variables|Sounds|GetSound_ScopedInLoop)
        (:then
          (bind _returnvalue_2 (Audio|SpawnSoundAttached _sound_scopedinloop (Variables|Visual|GetWeaponMesh)))
          (Variables|Default|SetScopedInLoopComp _returnvalue_2))
        (:else)))
    (:then_2
      (bind _scopezoomoutcomp (Variables|Default|GetScopeZoomOutComp)
        (:then
          (Audio|Components|Audio|FadeOut _scopezoomoutcomp 0.25 0.0 "Linear" :!class /Script/Engine.AudioComponent))
        (:else)))))

(fn StopScopedAudio
  (Utilities|FlowControl|Sequence
    (:then_0
      (bind _scopedinloopcomp (Variables|Default|GetScopedInLoopComp)
        (:then
          (Audio|Components|Audio|FadeOut _scopedinloopcomp 0.25 0.0 "Linear" :!class /Script/Engine.AudioComponent))
        (:else)))
    (:then_1
      (bind _scopezoomincomp (Variables|Default|GetScopeZoomInComp)
        (:then
          (Audio|Components|Audio|FadeOut _scopezoomincomp 0.0 0.0 "Linear" :!class /Script/Engine.AudioComponent))
        (:else)))
    (:then_2
      (if (Variables|/Script/Engine.SceneComponent|GetbHiddenInGame (Variables|Visual|GetWeaponMesh))
        (bind _sound_scopezoomout (Variables|Sounds|GetSound_ScopeZoomOut)
          (:then
            (bind _returnvalue (Audio|SpawnSoundAttached _sound_scopezoomout (Variables|Visual|GetWeaponMesh)))
            (Variables|Default|SetScopeZoomOutComp _returnvalue))
          (:else))))))

(fn UpdateShellEmittersFX
  (Utilities|FlowControl|Sequence
    (:then_0
      ;[1 1 1] Looping Shell Emitters
      (Effects|Components|ParticleSystem|SetEmitterEnable (Variables|Effects%28Shells%29|GetShells%28empty%29) "Small Loop" (and (Variables|Effects%28Shells%29|GetSmallShells) (Effects|ShouldPlayPersistentFireFX)))
      (Effects|Components|ParticleSystem|SetEmitterEnable (Variables|Effects%28Shells%29|GetShells%28empty%29) "Medium Loop" (and (Variables|Effects%28Shells%29|GetMediumShells) (Effects|ShouldPlayPersistentFireFX)))
      (Effects|Components|ParticleSystem|SetEmitterEnable (Variables|Effects%28Shells%29|GetShells%28empty%29) "Large Loop" (and (Variables|Effects%28Shells%29|GetLargeShells) (Effects|ShouldPlayPersistentFireFX)))
      (Effects|Components|ParticleSystem|SetEmitterEnable (Variables|Effects%28Shells%29|GetShells%28empty%29) "Energy Cell Loop" (and (Variables|Effects%28Shells%29|GetEnergyShells) (Effects|ShouldPlayPersistentFireFX)))
      ;[1 1 1] Single Shell Emitters
      (Effects|Components|ParticleSystem|SetEmitterEnable (Variables|Effects%28Shells%29|GetShells%28empty%29) "Small Single" (and (Variables|Effects%28Shells%29|GetSmallShells) (not (Effects|ShouldPlayPersistentFireFX))))
      (Effects|Components|ParticleSystem|SetEmitterEnable (Variables|Effects%28Shells%29|GetShells%28empty%29) "Medium Single" (and (Variables|Effects%28Shells%29|GetMediumShells) (not (Effects|ShouldPlayPersistentFireFX))))
      (Effects|Components|ParticleSystem|SetEmitterEnable (Variables|Effects%28Shells%29|GetShells%28empty%29) "Large Single" (and (Variables|Effects%28Shells%29|GetLargeShells) (not (Effects|ShouldPlayPersistentFireFX))))
      (Effects|Components|ParticleSystem|SetEmitterEnable (Variables|Effects%28Shells%29|GetShells%28empty%29) "Shotgun Single" (and (Variables|Effects%28Shells%29|GetShotgunShells) (not (Effects|ShouldPlayPersistentFireFX))))
      (Effects|Components|ParticleSystem|SetEmitterEnable (Variables|Effects%28Shells%29|GetShells%28empty%29) "Energy Cell Single" (and (Variables|Effects%28Shells%29|GetEnergyShells) (not (Effects|ShouldPlayPersistentFireFX)))))
    (:then_1
      (if (Variables|Effects%28Shells%29|GetUseShellsOnReload?)
        (Effects|Components|ParticleSystem|SetFloatParameter (Variables|Effects%28Shells%29|GetShells%28empty%29) "BurstScale" (Math|Conversions|ToFloat(Integer) (Utilities|ClampMinInteger (CallFunction|GetShellReloadCounter))))
        (CallFunction|SetShellReloadCounter -1)))))

(fn ConstructionScript
  (Effects|Components|ParticleSystem|SetAutoAttachmentParameters (Variables|Effects|GetMuzzle(Empty)) (Variables|Visual|GetWeaponMesh) (Effects|GetMuzzleSocketName 0) "KeepRelative" "KeepRelative" "KeepRelative")
  (Effects|Components|ParticleSystem|SetTemplate (Variables|Effects|GetMuzzle(Empty)) (Variables|Effects|GetMuzzleParticleSystem))
  ;[1 1 1] Reload Particles
  (if (Variables|Effects|GetUse%20Reload%20Particles)
    (Utilities|IsValid (Variables|Effects|GetReload_ParticleSystem)
      (:Is%20Valid
        (Effects|Components|ParticleSystem|SetTemplate (Variables|Default|GetReload(Empty)) (Variables|Effects|GetReload_ParticleSystem))
        (Transformation|AttachComponentToComponent (Variables|Default|GetReload(Empty)) (Variables|Visual|GetWeaponMesh) (Variables|Effects|GetReloadSocketName) "KeepRelative" "KeepRelative" "KeepRelative" true))
      (:Is%20Not%20Valid
        _)))
  ;[1 1 1] Wind Particle Setup
  ;; Set Is Wind Enabled
  (bind _returnvalue (Rendering|Material|GetScalarParameterValue "/Game/Packages/Fortress_SharedMaterials/GlobalMaterialParameters/FortniteMaterialParameters.FortniteMaterialParameters" "Water Interaction Enabled" :!class /Script/Engine.KismetMaterialLibrary))
  (bind _output_get (Variables|Default|SetIs%20Wind%20Enabled (> _returnvalue 0.5)))
  (if (not _output_get)
    (Components|DestroyComponent (Variables|Effects|GetMuzzleWindParticleSpawned)))
  (if (Variables|Default|GetIs%20Wind%20Enabled)
    (bind _returnvalue_2 (AddComponent|NODE_AddParticleSystemComponent-1))
    (Effects|Components|ParticleSystem|SetTemplate _returnvalue_2 (Variables|Effects|GetMuzzleWindParticleSystem))
    (bind _output_get_2 (Variables|Effects|SetMuzzleWindParticleSpawned _returnvalue_2))
    (Transformation|AttachComponentToComponent _output_get_2 (Variables|Visual|GetWeaponMesh) (Effects|GetMuzzleSocketName 0) "KeepRelative" "KeepRelative" "KeepRelative" true))
  ;[1 1 1] Shells FX setup
  (if (Math|Boolean|ORBoolean (Variables|Effects%28Shells%29|GetUseShellsOnFire?) (Variables|Effects%28Shells%29|GetUseShellsOnReload?) (Variables|Effects%28Shells%29|GetUseShellsOnPump?))
    (CallFunction|SetupShellFX)))
