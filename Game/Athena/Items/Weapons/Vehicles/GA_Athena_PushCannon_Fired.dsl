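; Gameplay ability for the vehicle Push Cannon that triggers its reloading gameplay cue when the cannon fires.
; On activation it runs generic setup, stores the player pawn and waits for the GameplayEffect.PushCannon.Fired event.
; On that event it adds the reloading cue to the owner, passing the event magnitude as raw magnitude, then ends itself.
; The end and passive-activation-failure overrides only call their parents, so cue removal is not handled here.
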(event Ability|EventActivateAbility
  (|CallParentFunction)
  (bind (_event_data _fort_player_pawn) (Setup|SetupGenericAbility false false)
    (:Output
      (Variables|Default|SetPlayerPawn _fort_player_pawn)
      (bind (_asynctaskproxy _payload) (Ability|Tasks|WaitGameplayEvent "(TagName=\"GameplayEffect.PushCannon.Fired\")")
        (:EventReceived
          (Ability|AddGameplayCueWithParamsToOwner (Variables|Default|GetReloadingCue) (Ability|GameplayCue|MakeGameplayCueParameters :NormalizedMagnitude 0.0 :RawMagnitude (.EventMagnitude (Utilities|Struct|BreakGameplayEventData _payload)) :EffectContext "()" :AggregatedSourceTags "(GameplayTags=)" :AggregatedTargetTags "(GameplayTags=)" :Location "0,0,0" :Normal "0,0,0" :GameplayEffectLevel 1 :AbilityLevel 1 :bReplicateLocationWhenUsingMinimalRepProxy false))
          (Ability|EndAbility))))))

(event Ability|EventOnEndAbility
  (params (bWasCancelled :type Boolean))
  (|CallParentFunction bWasCancelled))

(event Ability|EventFailedToActivatePassiveAbility
  (params (ActorInfo :type Gameplay%20Ability%20Actor%20Info%20Structure))
  (|CallParentFunction))
