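; Gameplay ability for a player fired from a push cannon, smashing through buildings until impact.
; Activation shrinks the capsule, blocks abilities, pushes a camera mode and repeats velocity-scaled capsule traces that destroy buildings, limited once a launch-area timer expires.
; Hit or landing triggers an impact cue and explosion that knocks back targets and pops the player upward, with damage for directly hit enemies.
; Tag/event watchers, a falling-state check and a timeout handle early termination.
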(event Ability|EventActivateAbility
  (bind _playerpawn (Variables|Default|GetPlayerPawn))
  (bind _asfort_player_pawn_athena (Utilities|Casting|CastToFortPlayerPawnAthena (Ability|GetAvatarActorfromActorInfo))
    (:then
      (bind _asfort_ability_system_component (Utilities|Casting|CastToFortAbilitySystemComponent (Ability|GetAbilitySystemComponent _asfort_player_pawn_athena)))
      (Variables|Default|SetAbilitySystem _asfort_ability_system_component)
      (Tags|BlockAbilitiesWithTags _asfort_ability_system_component (Variables|Default|GetBlockedAbilityTags))
      (bind _output_get (Variables|Default|SetPlayerPawn _asfort_player_pawn_athena))
      (CallFunction|ShrinkCollision _output_get)
      (if (not (Vehicle|IsinVehicle _output_get))
        (Collision|BindEventtoOnActorHit _playerpawn (EventDispatchers|CreateEvent :!event Custom|EventOnHit))
        (EventDispatchers|BindEventtoOnPawnLanded _playerpawn (EventDispatchers|CreateEvent :!event Custom|EventOnLanded))
        (bind _returnvalue (Utilities|Time|SetTimerbyEvent (EventDispatchers|CreateEvent :!event Custom|Trace%20ON) 0.033)) ; Overlap Timer
        (Variables|Timers|SetBuildingDestroyTImer _returnvalue)
        (bind _returnvalue_2 (Utilities|Time|SetTimerbyEvent (EventDispatchers|CreateEvent :!event Custom|SetLaunchAreaCleared) (ScalableFloat|GetValueatLevel (Variables|CurveRows|GetLaunchAreaClearDuration)))) ; Overlap Timer
        (Variables|Timers|SetLaunchAreaClearedTimer _returnvalue_2)
        (Camera|PushCameraMode (Variables|Default|GetCameraMode))
        (CallFunction|WatchforEndEvents)))
    (:CastFailed
      (Ability|EndAbility))))

(event Ability|EventFailedToActivatePassiveAbility
  (params (ActorInfo :type Gameplay%20Ability%20Actor%20Info%20Structure))
  (|CallParentFunction))

(event Ability|EventOnEndAbility
  (params (bWasCancelled :type Boolean))
  (bind _playerpawn (Variables|Default|GetPlayerPawn))
  (CallFunction|ResetCollision)
  (Ability|RemoveGrantedbyEffect)
  (Utilities|Time|ClearandInvalidateTimerbyHandle (Variables|Timers|GetBuildingDestroyTImer))
  (Utilities|Time|ClearandInvalidateTimerbyHandle (Variables|Timers|GetLaunchAreaClearedTimer))
  (Utilities|Time|ClearandInvalidateTimerbyHandle (Variables|Timers|GetTimeoutTimer))
  (Ability|RemoveGameplayEffectFromOwnerWithGrantedTags "(GameplayTags=((TagName=\"Status.LowGravity.ShipCannon\")))")
  (Collision|UnbindEventfromOnActorHit _playerpawn (EventDispatchers|CreateEvent :!event Custom|EventOnHit))
  (EventDispatchers|UnbindEventfromOnPawnLanded _playerpawn (EventDispatchers|CreateEvent :!event Custom|EventOnLanded))
  (Tags|UnBlockAbilitiesWithTags (Variables|Default|GetAbilitySystem) (Variables|Default|GetBlockedAbilityTags) false)
  ;[0.7396 0.3577 0.03666] End all pending ability tasks.
  (for _array_element (Variables|Default|GetAbilityTasks)
    (Utilities|IsValid _array_element
      (:Is%20Valid
        (GameplayTasks|EndTask _array_element))))
  (Utilities|Array|Clear (Variables|Default|GetAbilityTasks)))

(event Custom|Trace%20ON
  (bind _outhits (Collision|MultiCapsuleTraceForObjects (Transformation|GetActorLocation (Variables|Default|GetPlayerPawn)) (+ (Transformation|GetActorLocation (Variables|Default|GetPlayerPawn)) (* (Transformation|GetVelocity (Variables|Default|GetPlayerPawn)) (* (Utilities|Time|GetWorldDeltaSeconds) (ScalableFloat|GetValueatLevel (Variables|CurveRows|GetTraceFrames))))) (ScalableFloat|GetValueatLevel (Variables|CurveRows|GetCapsuleRadius)) (ScalableFloat|GetValueatLevel (Variables|CurveRows|GetCapsuleHalfHeight)) (Variables|Default|GetDestroyObjectTypes) false (Variables|Default|GetJumpHitActors)))
  (CallFunction|DestroyBuilding _outhits (Variables|Default|GetbLaunchAreaClear)))

(event Custom|EventOnHit
  (params (SelfActor :type Actor%20Object%20Reference) (OtherActor :type Actor%20Object%20Reference) (NormalImpulse :type Vector) (Hit :type Hit%20Result%20Structure%20%28by%20ref%29))
  (bind (_bblockinghit _binitialoverlap _time _distance _location _impactpoint _normal _impactnormal _physmat _hitactor) (Collision|BreakHitResult Hit))
  (Utilities|Time|ClearandInvalidateTimerbyHandle (Variables|Timers|GetBuildingDestroyTImer))
  (if (not (Variables|Default|GetbLostVelocity))
    (Variables|Default|SetbLostVelocity true)
    (CallFunction|HandleCollision _impactnormal)
    (bind _asfort_player_pawn_athena (Utilities|Casting|CastToFortPlayerPawnAthena _hitactor)
      (:then
        (if (not (Team|OnSameTeam _asfort_player_pawn_athena (Variables|Default|GetPlayerPawn)))
          (Environment|ApplyGlobalEnvironmentGameplayEffecttoActor _asfort_player_pawn_athena "/Game/Athena/Items/Weapons/Vehicles/GE_PushCannon_Launch_ImpactDamage.GE_PushCannon_Launch_ImpactDamage_C"))
        (CallFunction|DoExplosionandKnockback))
      (:CastFailed
        _))))

(event Custom|EventOnLanded
  (params (Hit :type Hit%20Result%20Structure%20%28by%20ref%29))
  (bind (_bblockinghit _binitialoverlap _time _distance _location _impactpoint _normal _impactnormal) (Collision|BreakHitResult Hit))
  (bind _playerpawn (Variables|Default|GetPlayerPawn))
  (Utilities|Time|ClearandInvalidateTimerbyHandle (Variables|Timers|GetBuildingDestroyTImer))
  (Utilities|Time|ClearandInvalidateTimerbyHandle (Variables|Timers|GetLaunchAreaClearedTimer))
  (Ability|RemoveGameplayEffectFromOwnerWithGrantedTags "(GameplayTags=((TagName=\"Status.LowGravity.ShipCannon\")))")
  (Collision|UnbindEventfromOnActorHit _playerpawn (EventDispatchers|CreateEvent :!event Custom|EventOnHit))
  (EventDispatchers|UnbindEventfromOnPawnLanded _playerpawn (EventDispatchers|CreateEvent :!event Custom|EventOnLanded))
  (if (Variables|Default|GetbLostVelocity)
    (Ability|RemoveGrantedbyEffect)
    (Ability|EndAbility)
    (else
      (Variables|Default|SetbLostVelocity true)
      (CallFunction|HandleCollision _impactnormal)
      (CallFunction|DoExplosionandKnockback)
      (Ability|RemoveGrantedbyEffect)
      (Ability|EndAbility))))

(event Custom|SetLaunchAreaCleared
  (Variables|Default|SetbLaunchAreaClear true))

(event Custom|DoExplosionAndKnockback
  (bind (_asynctaskproxy _targetdata _applicationtag) (Ability|Tasks|DoTargetSelection "None" "(TagName=\"Targeting.ApplicationTag.0\")" true)
    (:Targeted
      (Ability|ApplyGameplayEffectContainer _targetdata _applicationtag 0)
      (for _array_element (Ability|TargetData|GetActorsfromTargetData _targetdata)
        (bind _asfort_player_pawn_athena (Utilities|Casting|CastToFortPlayerPawnAthena _array_element))
        (Character|LaunchCharacter _asfort_player_pawn_athena (Math|Vector|MakeVector (.x (* (Math|Vector|Normalize (- (Transformation|GetActorLocation _array_element) (Transformation|GetActorLocation (Variables|Default|GetPlayerPawn)))) (ScalableFloat|GetValueatLevel (Variables|CurveRows|GetEnemyLaunchVelocity)))) (.y (* (Math|Vector|Normalize (- (Transformation|GetActorLocation _array_element) (Transformation|GetActorLocation (Variables|Default|GetPlayerPawn)))) (ScalableFloat|GetValueatLevel (Variables|CurveRows|GetEnemyLaunchVelocity)))) (Math|Float|Max(Float) (.z (* (Math|Vector|Normalize (- (Transformation|GetActorLocation _array_element) (Transformation|GetActorLocation (Variables|Default|GetPlayerPawn)))) (ScalableFloat|GetValueatLevel (Variables|CurveRows|GetEnemyLaunchVelocity)))) 5e+02)) true true))
      (Character|LaunchCharacter (Variables|Default|GetPlayerPawn) (Math|Vector|MakeVector 0.0 0.0 (ScalableFloat|GetValueatLevel (Variables|CurveRows|GetSelfVerticalLaunchVelocity))) true true)
      (Tags|UnBlockAbilitiesWithTags (Variables|Default|GetAbilitySystem) (Variables|Default|GetBlockedAbilityTags) false))))

(event Custom|WatchForEndEvents
  (bind _abilitytasks (Variables|Default|GetAbilityTasks))
  ;[0.8073 0.357 0.05078] WatchForEndEvents - Handle any events that may force this ability to end prematurely and force the ability to end after a set maximum time.
  ; Events that cause this ability to end early include: parachuting, creative edit mode flying, ziplining, entering a vent/geyser, being launched, entering a vehicle, entering a skytube, swimming.
  (Utilities|FlowControl|Sequence
    (:then_0
      (bind _asynctaskproxy (Ability|Tasks|WaitGameplayEvent "(TagName=\"Event.Parachute.Open\")" 0 true)
        (:then
          (Utilities|Array|Add _abilitytasks _asynctaskproxy))
        (:EventReceived
          _)))
    (:then_1
      (bind _asynctaskproxy_2 (Ability|Tasks|WaitGameplayTagAdd "(TagName=\"Gameplay.Action.Player.Creative.Flying\")" 0 true)
        (:then
          (Utilities|Array|Add _abilitytasks _asynctaskproxy_2))
        (:Added
          _)))
    (:then_2
      (bind _asynctaskproxy_3 (Ability|Tasks|WaitGameplayTagAdd "(TagName=\"Abilities.Generic.Athena.ZipLine\")" 0 true)
        (:then
          (Utilities|Array|Add _abilitytasks _asynctaskproxy_3))
        (:Added
          (Ability|EndAbility))))
    (:then_3
      (bind _asynctaskproxy_4 (Ability|Tasks|WaitGameplayEvent "(TagName=\"Abilities.Generic.Vent.Enter\")" 0 true)
        (:then
          (Utilities|Array|Add _abilitytasks _asynctaskproxy_4))
        (:EventReceived
          _)))
    (:then_4
      (bind _asynctaskproxy_5 (Ability|Tasks|WaitGameplayTagAdd "(TagName=\"Gameplay.Action.Player.Launched\")" 0 true)
        (:then
          (Utilities|Array|Add _abilitytasks _asynctaskproxy_5))
        (:Added
          _)))
    (:then_5
      (bind _asynctaskproxy_6 (Ability|Tasks|WaitGameplayEvent "(TagName=\"Event.Vehicle.Entered\")" 0 true)
        (:then
          (Utilities|Array|Add _abilitytasks _asynctaskproxy_6))
        (:EventReceived
          _)))
    (:then_6
      (bind _asynctaskproxy_7 (Ability|Tasks|WaitGameplayTagAdd "(TagName=\"Gameplay.Status.Skytubing\")" 0 true)
        (:then
          (Utilities|Array|Add _abilitytasks _asynctaskproxy_7))
        (:Added
          _)))
    (:then_7
      (bind _asynctaskproxy_8 (Ability|Tasks|WaitGameplayTagAdd "(TagName=\"Gameplay.Action.Player.SurfaceSwimming\")" 0 true)
        (:then
          (Utilities|Array|Add _abilitytasks _asynctaskproxy_8))
        (:Added
          _)))
    (:then_8
      (Utilities|Time|SetTimerbyEvent (delegate Custom|CheckForInvalidState) 1.0)
      (CallFunction|CheckforInvalidState))
    (:then_9
      (bind _returnvalue (Utilities|Time|SetTimerbyEvent (EventDispatchers|CreateEvent :!event Custom|TimeoutEndAbility) (ScalableFloat|GetValueatLevel (Variables|CurveRows|GetTimeoutDuration))))
      (Variables|Timers|SetTimeoutTimer _returnvalue))))

(event Custom|CheckForInvalidState
  ;[0.8073 0.357 0.05078] WatchForEndEvents - Handle any events that may force this ability to end prematurely and force the ability to end after a set maximum time.
  ; Events that cause this ability to end early include: parachuting, creative edit mode flying, ziplining, entering a vent/geyser, being launched, entering a vehicle, entering a skytube, swimming.
  (if (not (Math|Boolean|ANDBoolean (AI|Components|NavMovement|IsFalling (Variables|/Script/Engine.Character|GetCharacterMovement (Variables|Default|GetPlayerPawn))) (not (Parachute|IsSkydiving (Variables|Default|GetPlayerPawn))) (not (Parachute|IsParachuteOpen (Variables|Default|GetPlayerPawn)))))
    (Ability|EndAbility)))

(event Custom|TimeoutEndAbility
  ;[0.8073 0.357 0.05078] WatchForEndEvents - Handle any events that may force this ability to end prematurely and force the ability to end after a set maximum time.
  ; Events that cause this ability to end early include: parachuting, creative edit mode flying, ziplining, entering a vent/geyser, being launched, entering a vehicle, entering a skytube, swimming.
  (Ability|EndAbility))

(fn DestroyBuilding
  (params (BuildingActor :type Array%20of%20Hit%20Result%20Structures) (DestroySingleBuilding :type Boolean))
  (bind _hitactor (Variables|Default|GetHitActor))
  (Utilities|Array|Add (Variables|Default|GetJumpHitActors) (Variables|Default|GetPlayerPawn))
  (bind _array_element (Utilities|Array|ForEachLoopwithBreak BuildingActor)
    (:LoopBody
      (if (Variables|/Script/Engine.Actor|GetbCanBeDamaged (.HitActor (Collision|BreakHitResult _array_element)))
        (Variables|Default|SetHitActor (.HitActor (Collision|BreakHitResult _array_element)))
        (bind _bdidactorhavetags (Class|AthenaQuestFunctionLibrary|AthenaCheckActorHasTags _hitactor "(GameplayTags=((TagName=\"Deimos.Spawner\")))"))
        (if (not _bdidactorhavetags)
          (bind _bdidactorhavetags_2 (Class|AthenaQuestFunctionLibrary|AthenaCheckActorHasTags _hitactor "(GameplayTags=((TagName=\"Item.AppleSun\")))" true))
          (if (not _bdidactorhavetags_2)
            (Utilities|Casting|CastToAthenaCreativeRift _hitactor
              (:CastFailed
                (Utilities|Casting|CastToBuildingActor _hitactor
                  (:then
                    _)
                  (:CastFailed
                    (Utilities|Casting|CastToBuildingProp _hitactor
                      (:then
                        _)
                      (:CastFailed
                        (Utilities|Casting|CastToAthena_Prop_ParentBuildingContainerBlueprint _hitactor)))))
                (Utilities|Casting|CastToAthenaBarrierObjective (Variables|Default|GetHitActor)
                  (:CastFailed
                    (GameplayEffects|ApplyGameplayEffectToTarget :self (Ability|GetAbilitySystemComponent (Variables|Default|GetPlayerPawn)) :GameplayEffectClass (Variables|Default|GetBuildingDestroyGE) :Target (Ability|GetAbilitySystemComponent (Variables|Default|GetHitActor)) :Level 0.0)
                    (Development|PrintString "GA destroy building")
                    (Game|Feedback|ClientPlayForceFeedback (Fortnite|GetFortPlayerControllerfromActor (Variables|Default|GetPlayerPawn)) "/Game/Weapons/ForceFeedback/Explosions/FF_Generic_Explosion_Far.FF_Generic_Explosion_Far" "None" false false false)
                    (if (Variables|GetDestroySingleBuilding)
                      (LocalVariables|SetBuildingPieceDestroyed true))))))))))
    (:Completed
      (if (not (Math|Boolean|ANDBoolean (Utilities|IsArrayNonEmpty BuildingActor) (Variables|GetDestroySingleBuilding) (LocalVariables|GetBuildingPieceDestroyed)))
        (bind _returnvalue (Utilities|Time|SetTimerbyEvent (EventDispatchers|CreateEvent :!event Trace%20ON) 0.033)) ; Overlap Timer
        (Variables|Timers|SetBuildingDestroyTImer _returnvalue)))))

(fn HandleCollision
  (params (Normal :type Vector))
  (bind _playerpawn (Variables|Default|GetPlayerPawn))
  (Utilities|FlowControl|DoOnce false
    (:Completed
      (Ability|RemoveGameplayEffectFromOwnerWithGrantedTags "(GameplayTags=((TagName=\"Status.LowGravity.ShipCannon\")))")
      (Camera|PopCameraMode)
      (Game|Feedback|ClientStopCameraShake (Fortnite|GetFortPlayerControllerfromActor _playerpawn) "/Game/Athena/Environments/DudeBro/B_DudeBro_VortexLoop_Shake.B_DudeBro_VortexLoop_Shake_C" true)
      (Ability|ExecuteGameplayCueWithParamsOnOwner (Variables|Default|GetImpactCue) (Ability|GameplayCue|MakeGameplayCueParameters :NormalizedMagnitude 0.0 :RawMagnitude 0.0 :EffectContext "()" :AggregatedSourceTags "(GameplayTags=)" :AggregatedTargetTags "(GameplayTags=)" :Location (Math|Vector|MakeVectorNetQuantize10 (.x (Transformation|GetActorLocation _playerpawn)) (.y (Transformation|GetActorLocation _playerpawn)) (.z (Transformation|GetActorLocation _playerpawn))) :Normal (Math|Vector|MakeVectorNetQuantizeNormal (.x Normal) (.y Normal) (.z Normal)) :Instigator _playerpawn :EffectCauser _playerpawn :GameplayEffectLevel 0 :AbilityLevel 1 :bReplicateLocationWhenUsingMinimalRepProxy false))
      (Networking|ForceNetUpdate (Variables|Default|GetPlayerPawn))
      (Collision|UnbindEventfromOnActorHit (Variables|Default|GetPlayerPawn) (EventDispatchers|CreateEvent :!event EventOnHit)))))

(fn ResetCollision
  (Components|Capsule|SetCapsuleHalfHeight (Variables|/Script/Engine.Character|GetCapsuleComponent (Variables|Default|GetPlayerPawn)) (Variables|Default|GetCachedCapsuleHalfHeight) true))

(fn ShrinkCollision
  (params (PlayerPawn :type Character%20Object%20Reference))
  (bind _capsulecomponent (Variables|/Script/Engine.Character|GetCapsuleComponent PlayerPawn))
  (Variables|Default|SetCachedCapsuleHalfHeight (Variables|/Script/Engine.CapsuleComponent|GetCapsuleHalfHeight _capsulecomponent))
  (Components|Capsule|SetCapsuleHalfHeight _capsulecomponent (Variables|/Script/Engine.CapsuleComponent|GetCapsuleRadius _capsulecomponent) true))
