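; Gameplay ability for the Ostrich vehicle's missile launcher, which triggers the ranged weapon's secondary fire on the server.
; When activated from a gameplay event, it casts the avatar to an Athena player pawn and the pawn's current weapon to the Ostrich ranged weapon.
; If that weapon has authority, it calls the weapon's server secondary-fire activation, passing itself and the origin location from the event's target data.
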(event Ability|EventActivateAbilityFromEvent
  (params (EventData :type Gameplay%20Event%20Data%20Structure%20%28by%20ref%29))
  (bind (_eventtag _instigator _target _optionalobject _optionalobject2 _contexthandle _instigatortags _targettags _eventmagnitude _targetdata) (Utilities|Struct|BreakGameplayEventData EventData))
  ;[1 1 1] EVENT -> Activate Ability From Event
  (bind _asplayer_pawn_athena (Utilities|Casting|CastToPlayerPawn_Athena (Ability|GetAvatarActorfromActorInfo)))
  (bind _asfort_weapon_ranged_ostrich (Utilities|Casting|CastToFortWeaponRanged_Ostrich (Variables|/Script/FortniteGame.FortPawn|GetCurrentWeapon _asplayer_pawn_athena)))
  (if (Networking|HasAuthority _asfort_weapon_ranged_ostrich)
    (Gameplay|OnServerActivateSecondaryFire _asfort_weapon_ranged_ostrich self (.location (Ability|TargetData|GetTargetDataOrigin _targetdata)))))
